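# Physics asset for the Moss Knight player body cosmetic, simulating flowers and leaves on the shoulders, arms and right hand.
# Flowers use boxes and leaves use capsules, each joined to its tiny "_master" capsule by narrow swing and twist limits.
# Every master, including arm and hand ones, constrains to its side's clavicle capsule.
# These links lock swing2 and twist, except right shoulder leaf 19 and left 015, which rotate freely.
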PhysicsAsset /BRCosmetics/Characters/Player/Male/Medium/Bodies/M_MED_Moss_Knight/Meshes/M_MED_Moss_Knight_Flower_Leaf_Physics
bodies {
 clavicle_l [{shapeType=Capsule, length=15.3716, radius=8.3347, center=<1.2541 -0.0001 0>, rotation=<rot 90 0 0>}]
 clavicle_r [{shapeType=Capsule, length=14.4619, radius=8.3347, center=<-2.8221 0 0>, rotation=<rot 90 0 0>}]
 dyn_l_arm_flower_01 [{shapeType=Box, extentX=1.3181, extentY=3.0186, extentZ=3.0186, center=<0.5228 0 0>}]
 dyn_l_arm_flower_01_master [{shapeType=Capsule, length=0.505, radius=0.505, rotation=<rot 0 -90 0>}]
 dyn_l_arm_flower_02 [{shapeType=Box, extentX=1.3181, extentY=3.0186, extentZ=3.0186, center=<0.5228 0 0>}]
 dyn_l_arm_flower_02_master [{shapeType=Capsule, length=0.505, radius=0.505, rotation=<rot 0 -90 0>}]
 dyn_l_arm_flower_03 [{shapeType=Box, extentX=1.3181, extentY=3.0186, extentZ=3.0186, center=<0.5228 0 0>}]
 dyn_l_arm_flower_03_master [{shapeType=Capsule, length=0.505, radius=0.505, rotation=<rot 0 -90 0>}]
 dyn_l_shoulder_flower_01 [{shapeType=Box, extentX=1.3181, extentY=3.0186, extentZ=3.0186, center=<0.5228 0 0>}]
 dyn_l_shoulder_flower_01_master [{shapeType=Capsule, length=0.505, radius=0.505, rotation=<rot 0 -90 0>}]
 dyn_l_shoulder_flower_02 [{shapeType=Box, extentX=1.3181, extentY=3.0186, extentZ=3.0186, center=<0.5228 0 0>}]
 dyn_l_shoulder_flower_02_master [{shapeType=Capsule, length=0.505, radius=0.505, rotation=<rot 0 -90 0>}]
 dyn_l_shoulder_flower_03 [{shapeType=Box, extentX=1.3181, extentY=3.0186, extentZ=3.0186, center=<0.5228 0 0>}]
 dyn_l_shoulder_flower_03_master [{shapeType=Capsule, length=0.505, radius=0.505, rotation=<rot 0 -90 0>}]
 dyn_l_shoulder_flower_04 [{shapeType=Box, extentX=1.3181, extentY=3.0186, extentZ=3.0186, center=<0.5228 0 0>}]
 dyn_l_shoulder_flower_04_master [{shapeType=Capsule, length=0.505, radius=0.505, rotation=<rot 0 -90 0>}]
 dyn_l_shoulder_flower_05 [{shapeType=Box, extentX=1.3181, extentY=3.0186, extentZ=3.0186, center=<0.5228 0 0>}]
 dyn_l_shoulder_flower_05_master [{shapeType=Capsule, length=0.505, radius=0.505, rotation=<rot 0 -90 0>}]
 dyn_l_shoulder_flower_06 [{shapeType=Box, extentX=1.3181, extentY=3.0186, extentZ=3.0186, center=<0.5228 0 0>}]
 dyn_l_shoulder_flower_06_master [{shapeType=Capsule, length=0.505, radius=0.505, rotation=<rot 0 -90 0>}]
 dyn_l_shoulder_flower_07 [{shapeType=Box, extentX=1.3181, extentY=3.0186, extentZ=3.0186, center=<0.5228 0 0>}]
 dyn_l_shoulder_flower_07_master [{shapeType=Capsule, length=0.505, radius=0.505, rotation=<rot 0 -90 0>}]
 dyn_l_shoulder_flower_08 [{shapeType=Box, extentX=1.3181, extentY=5.0341, extentZ=5.0341, center=<0.5228 0 0>}]
 dyn_l_shoulder_flower_08_master [{shapeType=Capsule, length=0.505, radius=0.505, rotation=<rot 0 -90 0>}]
 dyn_l_shoulder_flower_09 [{shapeType=Box, extentX=1.3181, extentY=3.0186, extentZ=3.0186, center=<0.5228 0 0>}]
 dyn_l_shoulder_flower_09_master [{shapeType=Capsule, length=0.505, radius=0.505, rotation=<rot 0 -90 0>}]
 dyn_l_shoulder_leaf_01 [{shapeType=Capsule, length=1.906, radius=1, center=<2.2033 0.0001 0>, rotation=<rot 90 0 0>}]
 dyn_l_shoulder_leaf_010 [{shapeType=Capsule, length=1.906, radius=1, center=<2.2033 0.0001 0>, rotation=<rot 90 0 0>}]
 dyn_l_shoulder_leaf_010_master [{shapeType=Capsule, length=0.505, radius=0.505, rotation=<rot 0 -90 0>}]
 dyn_l_shoulder_leaf_011 [{shapeType=Capsule, length=1.906, radius=1, center=<2.2033 0.0001 0>, rotation=<rot 90 0 0>}]
 dyn_l_shoulder_leaf_011_master [{shapeType=Capsule, length=0.505, radius=0.505, rotation=<rot 0 -90 0>}]
 dyn_l_shoulder_leaf_012 [{shapeType=Capsule, length=1.906, radius=1, center=<2.2033 0.0001 0>, rotation=<rot 90 0 0>}]
 dyn_l_shoulder_leaf_012_master [{shapeType=Capsule, length=0.505, radius=0.505, rotation=<rot 0 -90 0>}]
 dyn_l_shoulder_leaf_013 [{shapeType=Capsule, length=1.906, radius=1, center=<2.2033 0.0001 0>, rotation=<rot 90 0 0>}]
 dyn_l_shoulder_leaf_013_master [{shapeType=Capsule, length=0.505, radius=0.505, rotation=<rot 0 -90 0>}]
 dyn_l_shoulder_leaf_014 [{shapeType=Capsule, length=1.906, radius=1, center=<2.2033 0.0001 0>, rotation=<rot 90 0 0>}]
 dyn_l_shoulder_leaf_014_master [{shapeType=Capsule, length=0.505, radius=0.505, rotation=<rot 0 -90 0>}]
 dyn_l_shoulder_leaf_015 [{shapeType=Capsule, length=1.906, radius=1, center=<2.2033 0.0001 0>, rotation=<rot 90 0 0>}]
 dyn_l_shoulder_leaf_015_master [{shapeType=Capsule, length=0.505, radius=0.505, rotation=<rot 0 -90 0>}]
 dyn_l_shoulder_leaf_01_master [{shapeType=Capsule, length=0.505, radius=0.505, rotation=<rot 0 -90 0>}]
 dyn_l_shoulder_leaf_02 [{shapeType=Capsule, length=1.906, radius=1, center=<2.2033 0.0001 0>, rotation=<rot 90 0 0>}]
 dyn_l_shoulder_leaf_02_master [{shapeType=Capsule, length=0.505, radius=0.505, rotation=<rot 0 -90 0>}]
 dyn_l_shoulder_leaf_03 [{shapeType=Capsule, length=1.906, radius=1, center=<2.2033 0.0001 0>, rotation=<rot 90 0 0>}]
 dyn_l_shoulder_leaf_03_master [{shapeType=Capsule, length=0.505, radius=0.505, rotation=<rot 0 -90 0>}]
 dyn_l_shoulder_leaf_04 [{shapeType=Capsule, length=1.906, radius=1, center=<2.2033 0.0001 0>, rotation=<rot 90 0 0>}]
 dyn_l_shoulder_leaf_04_master [{shapeType=Capsule, length=0.505, radius=0.505, rotation=<rot 0 -90 0>}]
 dyn_l_shoulder_leaf_05 [{shapeType=Capsule, length=1.906, radius=1, center=<2.2033 0.0001 0>, rotation=<rot 90 0 0>}]
 dyn_l_shoulder_leaf_05_master [{shapeType=Capsule, length=0.505, radius=0.505, rotation=<rot 0 -90 0>}]
 dyn_l_shoulder_leaf_06 [{shapeType=Capsule, length=1.906, radius=1, center=<2.2033 0.0001 0>, rotation=<rot 90 0 0>}]
 dyn_l_shoulder_leaf_06_master [{shapeType=Capsule, length=0.505, radius=0.505, rotation=<rot 0 -90 0>}]
 dyn_l_shoulder_leaf_07 [{shapeType=Capsule, length=2.1442, radius=1.5153, center=<3.1168 0 0>, rotation=<rot 90 0 0>}]
 dyn_l_shoulder_leaf_07_master [{shapeType=Capsule, length=0.505, radius=0.505, rotation=<rot 0 -90 0>}]
 dyn_l_shoulder_leaf_08 [{shapeType=Capsule, length=1.906, radius=1, center=<2.2033 0.0001 0>, rotation=<rot 90 0 0>}]
 dyn_l_shoulder_leaf_08_master [{shapeType=Capsule, length=0.505, radius=0.505, rotation=<rot 0 -90 0>}]
 dyn_l_shoulder_leaf_09 [{shapeType=Capsule, length=1.906, radius=1, center=<2.2033 0.0001 0>, rotation=<rot 90 0 0>}]
 dyn_l_shoulder_leaf_09_master [{shapeType=Capsule, length=0.505, radius=0.505, rotation=<rot 0 -90 0>}]
 dyn_r_arm_leaf_01 [{shapeType=Capsule, length=1.906, radius=1, center=<2.2033 0.0001 0>, rotation=<rot 90 0 0>}]
 dyn_r_arm_leaf_01_master [{shapeType=Capsule, length=0.505, radius=0.505, rotation=<rot 0 -90 0>}]
 dyn_r_arm_leaf_02 [{shapeType=Capsule, length=1.906, radius=1, center=<2.2033 0.0001 0>, rotation=<rot 90 0 0>}]
 dyn_r_arm_leaf_02_master [{shapeType=Capsule, length=0.505, radius=0.505, rotation=<rot 0 -90 0>}]
 dyn_r_arm_leaf_03 [{shapeType=Capsule, length=1.906, radius=1, center=<2.2033 0.0001 0>, rotation=<rot 90 0 0>}]
 dyn_r_arm_leaf_03_master [{shapeType=Capsule, length=0.505, radius=0.505, rotation=<rot 0 -90 0>}]
 dyn_r_arm_leaf_04 [{shapeType=Capsule, length=1.906, radius=1, center=<2.2033 0.0001 0>, rotation=<rot 90 0 0>}]
 dyn_r_arm_leaf_04_master [{shapeType=Capsule, length=0.505, radius=0.505, rotation=<rot 0 -90 0>}]
 dyn_r_hand_flower_01 [{shapeType=Box, extentX=1.3181, extentY=3.0186, extentZ=3.0186, center=<0.5228 0 0>}]
 dyn_r_hand_flower_01_master [{shapeType=Capsule, length=0.505, radius=0.505, rotation=<rot 0 -90 0>}]
 dyn_r_hand_flower_02 [{shapeType=Box, extentX=1.3181, extentY=3.0186, extentZ=3.0186, center=<0.5228 0 0>}]
 dyn_r_hand_flower_02_master [{shapeType=Capsule, length=0.505, radius=0.505, rotation=<rot 0 -90 0>}]
 dyn_r_shoulder_flower_01 [{shapeType=Box, extentX=1.3181, extentY=3.0186, extentZ=3.0186, center=<0.5228 0 0>}]
 dyn_r_shoulder_flower_01_master [{shapeType=Capsule, length=0.505, radius=0.505, rotation=<rot 0 -90 0>}]
 dyn_r_shoulder_flower_02 [{shapeType=Box, extentX=1.3181, extentY=3.0186, extentZ=3.0186, center=<0.5228 0 0>}]
 dyn_r_shoulder_flower_02_master [{shapeType=Capsule, length=0.505, radius=0.505, rotation=<rot 0 -90 0>}]
 dyn_r_shoulder_flower_03 [{shapeType=Box, extentX=1.3181, extentY=3.0186, extentZ=3.0186, center=<0.5228 0 0>}]
 dyn_r_shoulder_flower_03_master [{shapeType=Capsule, length=0.505, radius=0.505, rotation=<rot 0 -90 0>}]
 dyn_r_shoulder_flower_04 [{shapeType=Box, extentX=1.3181, extentY=3.0186, extentZ=3.0186, center=<0.5228 0 0>}]
 dyn_r_shoulder_flower_04_master [{shapeType=Capsule, length=0.505, radius=0.505, rotation=<rot 0 -90 0>}]
 dyn_r_shoulder_flower_05 [{shapeType=Box, extentX=1.3181, extentY=3.0186, extentZ=3.0186, center=<0.5228 0 0>}]
 dyn_r_shoulder_flower_05_master [{shapeType=Capsule, length=0.505, radius=0.505, rotation=<rot 0 -90 0>}]
 dyn_r_shoulder_leaf_01 [{shapeType=Capsule, length=1.906, radius=1, center=<2.2033 0.0001 0>, rotation=<rot 90 0 0>}]
 dyn_r_shoulder_leaf_01_master [{shapeType=Capsule, length=0.505, radius=0.505}]
 dyn_r_shoulder_leaf_02 [{shapeType=Capsule, length=1.906, radius=1, center=<2.2033 0.0001 0>, rotation=<rot 90 0 0>}]
 dyn_r_shoulder_leaf_02_master [{shapeType=Capsule, length=0.505, radius=0.505}]
 dyn_r_shoulder_leaf_03 [{shapeType=Capsule, length=1.906, radius=1, center=<2.2033 0.0001 0>, rotation=<rot 90 0 0>}]
 dyn_r_shoulder_leaf_03_master [{shapeType=Capsule, length=0.505, radius=0.505}]
 dyn_r_shoulder_leaf_04 [{shapeType=Capsule, length=1.906, radius=1, center=<2.2033 0.0001 0>, rotation=<rot 90 0 0>}]
 dyn_r_shoulder_leaf_04_master [{shapeType=Capsule, length=0.505, radius=0.505}]
 dyn_r_shoulder_leaf_05 [{shapeType=Capsule, length=1.906, radius=1, center=<2.2033 0.0001 0>, rotation=<rot 90 0 0>}]
 dyn_r_shoulder_leaf_05_master [{shapeType=Capsule, length=0.505, radius=0.505}]
 dyn_r_shoulder_leaf_06 [{shapeType=Capsule, length=1.906, radius=1, center=<2.2033 0.0001 0>, rotation=<rot 90 0 0>}]
 dyn_r_shoulder_leaf_06_master [{shapeType=Capsule, length=0.505, radius=0.505}]
 dyn_r_shoulder_leaf_07 [{shapeType=Capsule, length=1.906, radius=1, center=<2.2033 0.0001 0>, rotation=<rot 90 0 0>}]
 dyn_r_shoulder_leaf_07_master [{shapeType=Capsule, length=0.505, radius=0.505}]
 dyn_r_shoulder_leaf_08 [{shapeType=Capsule, length=1.906, radius=1, center=<2.2033 0.0001 0>, rotation=<rot 90 0 0>}]
 dyn_r_shoulder_leaf_08_master [{shapeType=Capsule, length=0.505, radius=0.505}]
 dyn_r_shoulder_leaf_09 [{shapeType=Capsule, length=1.906, radius=1, center=<2.2033 0.0001 0>, rotation=<rot 90 0 0>}]
 dyn_r_shoulder_leaf_09_master [{shapeType=Capsule, length=0.505, radius=0.505}]
 dyn_r_shoulder_leaf_10 [{shapeType=Capsule, length=1.906, radius=1, center=<2.2033 0.0001 0>, rotation=<rot 90 0 0>}]
 dyn_r_shoulder_leaf_10_master [{shapeType=Capsule, length=0.505, radius=0.505}]
 dyn_r_shoulder_leaf_11 [{shapeType=Capsule, length=1.906, radius=1, center=<2.2033 0.0001 0>, rotation=<rot 90 0 0>}]
 dyn_r_shoulder_leaf_11_master [{shapeType=Capsule, length=0.505, radius=0.505}]
 dyn_r_shoulder_leaf_12 [{shapeType=Capsule, length=1.906, radius=1, center=<2.2033 0.0001 0>, rotation=<rot 90 0 0>}]
 dyn_r_shoulder_leaf_12_master [{shapeType=Capsule, length=0.505, radius=0.505}]
 dyn_r_shoulder_leaf_13 [{shapeType=Capsule, length=1.906, radius=1, center=<2.2033 0.0001 0>, rotation=<rot 90 0 0>}]
 dyn_r_shoulder_leaf_13_master [{shapeType=Capsule, length=0.505, radius=0.505}]
 dyn_r_shoulder_leaf_14 [{shapeType=Capsule, length=1.906, radius=1, center=<2.2033 0.0001 0>, rotation=<rot 90 0 0>}]
 dyn_r_shoulder_leaf_14_master [{shapeType=Capsule, length=0.505, radius=0.505}]
 dyn_r_shoulder_leaf_15 [{shapeType=Capsule, length=1.906, radius=1, center=<2.2033 0.0001 0>, rotation=<rot 90 0 0>}]
 dyn_r_shoulder_leaf_15_master [{shapeType=Capsule, length=0.505, radius=0.505}]
 dyn_r_shoulder_leaf_16 [{shapeType=Capsule, length=1.906, radius=1, center=<2.2033 0.0001 0>, rotation=<rot 90 0 0>}]
 dyn_r_shoulder_leaf_16_master [{shapeType=Capsule, length=0.505, radius=0.505}]
 dyn_r_shoulder_leaf_17 [{shapeType=Capsule, length=1.906, radius=1, center=<2.2033 0.0001 0>, rotation=<rot 90 0 0>}]
 dyn_r_shoulder_leaf_17_master [{shapeType=Capsule, length=0.505, radius=0.505}]
 dyn_r_shoulder_leaf_18 [{shapeType=Capsule, length=1.906, radius=1, center=<2.2033 0.0001 0>, rotation=<rot 90 0 0>}]
 dyn_r_shoulder_leaf_18_master [{shapeType=Capsule, length=0.505, radius=0.505}]
 dyn_r_shoulder_leaf_19 [{shapeType=Capsule, length=1.906, radius=1, center=<2.2033 0.0001 0>, rotation=<rot 90 0 0>}]
 dyn_r_shoulder_leaf_19_master [{shapeType=Capsule, length=0.505, radius=0.505, rotation=<rot 0 -90 0>}]
}
constraints [
 {
  bone1Name=dyn_l_shoulder_leaf_01
  bone2Name=dyn_l_shoulder_leaf_01_master
  swing1LimitDegrees=10
  swing2LimitDegrees=17
  twistLimitDegrees=10
 }
 {
  bone1Name=dyn_l_shoulder_leaf_02
  bone2Name=dyn_l_shoulder_leaf_02_master
  swing1LimitDegrees=10
  swing2LimitDegrees=10
  twistLimitDegrees=10
 }
 {
  bone1Name=dyn_l_shoulder_leaf_03
  bone2Name=dyn_l_shoulder_leaf_03_master
  swing1LimitDegrees=10
  swing2LimitDegrees=17
  twistLimitDegrees=10
 }
 {
  bone1Name=dyn_l_shoulder_leaf_04
  bone2Name=dyn_l_shoulder_leaf_04_master
  swing1LimitDegrees=10
  swing2LimitDegrees=17
  twistLimitDegrees=10
 }
 {
  bone1Name=dyn_l_shoulder_leaf_05
  bone2Name=dyn_l_shoulder_leaf_05_master
  swing1LimitDegrees=10
  swing2LimitDegrees=10
  twistLimitDegrees=10
 }
 {
  bone1Name=dyn_l_shoulder_leaf_06
  bone2Name=dyn_l_shoulder_leaf_06_master
  swing1LimitDegrees=10
  swing2LimitDegrees=17
  twistLimitDegrees=10
 }
 {
  bone1Name=dyn_l_shoulder_leaf_07
  bone2Name=dyn_l_shoulder_leaf_07_master
  swing1LimitDegrees=10
  swing2LimitDegrees=10
  twistLimitDegrees=10
 }
 {
  bone1Name=dyn_l_shoulder_leaf_08
  bone2Name=dyn_l_shoulder_leaf_08_master
  swing1LimitDegrees=10
  swing2LimitDegrees=17
  twistLimitDegrees=10
 }
 {
  bone1Name=dyn_l_shoulder_leaf_09
  bone2Name=dyn_l_shoulder_leaf_09_master
  swing1LimitDegrees=15
  swing2LimitDegrees=20
  twistLimitDegrees=10
 }
 {
  bone1Name=dyn_l_shoulder_leaf_010
  bone2Name=dyn_l_shoulder_leaf_010_master
  swing1LimitDegrees=10
  swing2LimitDegrees=10
  twistLimitDegrees=10
 }
 {
  bone1Name=dyn_l_shoulder_leaf_011
  bone2Name=dyn_l_shoulder_leaf_011_master
  swing1LimitDegrees=10
  swing2LimitDegrees=10
  twistLimitDegrees=10
 }
 {
  bone1Name=dyn_l_shoulder_leaf_012
  bone2Name=dyn_l_shoulder_leaf_012_master
  swing1LimitDegrees=10
  swing2LimitDegrees=10
  twistLimitDegrees=10
 }
 {
  bone1Name=dyn_l_shoulder_leaf_013
  bone2Name=dyn_l_shoulder_leaf_013_master
  swing1LimitDegrees=10
  swing2LimitDegrees=10
  twistLimitDegrees=10
 }
 {
  bone1Name=dyn_l_shoulder_leaf_014
  bone2Name=dyn_l_shoulder_leaf_014_master
  swing1LimitDegrees=10
  swing2LimitDegrees=17
  twistLimitDegrees=10
 }
 {
  bone1Name=dyn_r_shoulder_leaf_18
  bone2Name=dyn_r_shoulder_leaf_18_master
  swing1LimitDegrees=10
  swing2LimitDegrees=17
  twistLimitDegrees=10
 }
 {
  bone1Name=dyn_r_shoulder_leaf_17
  bone2Name=dyn_r_shoulder_leaf_17_master
  swing1LimitDegrees=10
  swing2LimitDegrees=10
  twistLimitDegrees=10
 }
 {
  bone1Name=dyn_r_shoulder_leaf_16
  bone2Name=dyn_r_shoulder_leaf_16_master
  swing1LimitDegrees=15
  swing2LimitDegrees=20
  twistLimitDegrees=10
 }
 {
  bone1Name=dyn_r_shoulder_leaf_15
  bone2Name=dyn_r_shoulder_leaf_15_master
  swing1LimitDegrees=10
  swing2LimitDegrees=10
  twistLimitDegrees=10
 }
 {
  bone1Name=dyn_r_shoulder_leaf_14
  bone2Name=dyn_r_shoulder_leaf_14_master
  swing1LimitDegrees=10
  swing2LimitDegrees=10
  twistLimitDegrees=10
 }
 {
  bone1Name=dyn_r_shoulder_leaf_13
  bone2Name=dyn_r_shoulder_leaf_13_master
  swing1LimitDegrees=10
  swing2LimitDegrees=10
  twistLimitDegrees=10
 }
 {
  bone1Name=dyn_r_shoulder_leaf_12
  bone2Name=dyn_r_shoulder_leaf_12_master
  swing1LimitDegrees=10
  swing2LimitDegrees=10
  twistLimitDegrees=10
 }
 {
  bone1Name=dyn_r_shoulder_leaf_11
  bone2Name=dyn_r_shoulder_leaf_11_master
  swing1LimitDegrees=10
  swing2LimitDegrees=10
  twistLimitDegrees=10
 }
 {
  bone1Name=dyn_r_shoulder_leaf_10
  bone2Name=dyn_r_shoulder_leaf_10_master
  swing1LimitDegrees=10
  swing2LimitDegrees=10
  twistLimitDegrees=10
 }
 {
  bone1Name=dyn_r_shoulder_leaf_09
  bone2Name=dyn_r_shoulder_leaf_09_master
  swing1LimitDegrees=10
  swing2LimitDegrees=10
  twistLimitDegrees=10
 }
 {
  bone1Name=dyn_r_shoulder_leaf_08
  bone2Name=dyn_r_shoulder_leaf_08_master
  swing1LimitDegrees=10
  swing2LimitDegrees=10
  twistLimitDegrees=10
 }
 {
  bone1Name=dyn_r_shoulder_leaf_07
  bone2Name=dyn_r_shoulder_leaf_07_master
  swing1LimitDegrees=10
  swing2LimitDegrees=10
  twistLimitDegrees=10
 }
 {
  bone1Name=dyn_r_shoulder_leaf_06
  bone2Name=dyn_r_shoulder_leaf_06_master
  swing1LimitDegrees=10
  swing2LimitDegrees=10
  twistLimitDegrees=10
 }
 {
  bone1Name=dyn_r_shoulder_leaf_05
  bone2Name=dyn_r_shoulder_leaf_05_master
  swing1LimitDegrees=10
  swing2LimitDegrees=10
  twistLimitDegrees=10
 }
 {
  bone1Name=dyn_r_shoulder_leaf_04
  bone2Name=dyn_r_shoulder_leaf_04_master
  swing1LimitDegrees=10
  swing2LimitDegrees=10
  twistLimitDegrees=10
 }
 {
  bone1Name=dyn_r_shoulder_leaf_03
  bone2Name=dyn_r_shoulder_leaf_03_master
  swing1LimitDegrees=10
  swing2LimitDegrees=10
  twistLimitDegrees=10
 }
 {
  bone1Name=dyn_r_shoulder_leaf_02
  bone2Name=dyn_r_shoulder_leaf_02_master
  swing1LimitDegrees=15
  swing2LimitDegrees=20
  twistLimitDegrees=10
 }
 {
  bone1Name=dyn_r_shoulder_leaf_01
  bone2Name=dyn_r_shoulder_leaf_01_master
  swing1LimitDegrees=15
  swing2LimitDegrees=20
  twistLimitDegrees=10
 }
 {
  bone1Name=dyn_l_shoulder_flower_01
  bone2Name=dyn_l_shoulder_flower_01_master
  swing1LimitDegrees=10
  swing2LimitDegrees=10
  twistLimitDegrees=10
 }
 {
  bone1Name=dyn_l_shoulder_flower_02
  bone2Name=dyn_l_shoulder_flower_02_master
  swing1LimitDegrees=10
  swing2LimitDegrees=10
  twistLimitDegrees=10
 }
 {
  bone1Name=dyn_l_shoulder_flower_03
  bone2Name=dyn_l_shoulder_flower_03_master
  swing1LimitDegrees=10
  swing2LimitDegrees=10
  twistLimitDegrees=10
 }
 {
  bone1Name=dyn_l_shoulder_flower_04
  bone2Name=dyn_l_shoulder_flower_04_master
  swing1LimitDegrees=10
  swing2LimitDegrees=10
  twistLimitDegrees=10
 }
 {
  bone1Name=dyn_l_shoulder_flower_06
  bone2Name=dyn_l_shoulder_flower_06_master
  swing1LimitDegrees=10
  swing2LimitDegrees=10
  twistLimitDegrees=10
 }
 {
  bone1Name=dyn_l_shoulder_flower_09
  bone2Name=dyn_l_shoulder_flower_09_master
  swing1LimitDegrees=10
  swing2LimitDegrees=10
  twistLimitDegrees=10
 }
 {
  bone1Name=dyn_l_shoulder_flower_08
  bone2Name=dyn_l_shoulder_flower_08_master
  swing1LimitDegrees=10
  swing2LimitDegrees=10
  twistLimitDegrees=10
 }
 {
  bone1Name=dyn_l_shoulder_flower_07
  bone2Name=dyn_l_shoulder_flower_07_master
  swing1LimitDegrees=10
  swing2LimitDegrees=10
  twistLimitDegrees=10
 }
 {
  bone1Name=dyn_l_shoulder_flower_05
  bone2Name=dyn_l_shoulder_flower_05_master
  swing1LimitDegrees=10
  swing2LimitDegrees=10
  twistLimitDegrees=10
 }
 {
  bone1Name=dyn_r_shoulder_flower_05
  bone2Name=dyn_r_shoulder_flower_05_master
  swing1LimitDegrees=10
  swing2LimitDegrees=10
  twistLimitDegrees=10
 }
 {
  bone1Name=dyn_r_shoulder_flower_04
  bone2Name=dyn_r_shoulder_flower_04_master
  swing1LimitDegrees=10
  swing2LimitDegrees=10
  twistLimitDegrees=10
 }
 {
  bone1Name=dyn_r_shoulder_flower_03
  bone2Name=dyn_r_shoulder_flower_03_master
  swing1LimitDegrees=10
  swing2LimitDegrees=10
  twistLimitDegrees=10
 }
 {
  bone1Name=dyn_r_shoulder_flower_02
  bone2Name=dyn_r_shoulder_flower_02_master
  swing1LimitDegrees=10
  swing2LimitDegrees=10
  twistLimitDegrees=10
 }
 {
  bone1Name=dyn_r_shoulder_flower_01
  bone2Name=dyn_r_shoulder_flower_01_master
  swing1LimitDegrees=10
  swing2LimitDegrees=10
  twistLimitDegrees=10
 }
 {bone1Name=dyn_r_arm_leaf_04, bone2Name=dyn_r_arm_leaf_04_master, swing1LimitDegrees=10, swing2LimitDegrees=20, twistLimitDegrees=10}
 {bone1Name=dyn_r_arm_leaf_02, bone2Name=dyn_r_arm_leaf_02_master, swing1LimitDegrees=10, swing2LimitDegrees=10, twistLimitDegrees=10}
 {bone1Name=dyn_r_arm_leaf_03, bone2Name=dyn_r_arm_leaf_03_master, swing1LimitDegrees=10, swing2LimitDegrees=15, twistLimitDegrees=10}
 {bone1Name=dyn_r_arm_leaf_01, bone2Name=dyn_r_arm_leaf_01_master, swing1LimitDegrees=10, swing2LimitDegrees=15, twistLimitDegrees=10}
 {bone1Name=dyn_l_arm_flower_03, bone2Name=dyn_l_arm_flower_03_master, swing1LimitDegrees=10, swing2LimitDegrees=10, twistLimitDegrees=10}
 {bone1Name=dyn_l_arm_flower_02, bone2Name=dyn_l_arm_flower_02_master, swing1LimitDegrees=10, swing2LimitDegrees=10, twistLimitDegrees=10}
 {bone1Name=dyn_l_arm_flower_01, bone2Name=dyn_l_arm_flower_01_master, swing1LimitDegrees=10, swing2LimitDegrees=10, twistLimitDegrees=10}
 {bone1Name=dyn_r_hand_flower_02, bone2Name=dyn_r_hand_flower_02_master, swing1LimitDegrees=10, swing2LimitDegrees=10, twistLimitDegrees=10}
 {bone1Name=dyn_r_hand_flower_01, bone2Name=dyn_r_hand_flower_01_master, swing1LimitDegrees=10, swing2LimitDegrees=10, twistLimitDegrees=10}
 {
  bone1Name=dyn_r_shoulder_leaf_18_master
  bone2Name=clavicle_r
  swing1LimitDegrees=45
  swing2LimitDegrees=45
  swing2Motion=Locked
  twistLimitDegrees=45
  twistMotion=Locked
 }
 {
  bone1Name=dyn_r_shoulder_leaf_17_master
  bone2Name=clavicle_r
  swing1LimitDegrees=45
  swing2LimitDegrees=45
  swing2Motion=Locked
  twistLimitDegrees=45
  twistMotion=Locked
 }
 {
  bone1Name=dyn_r_shoulder_leaf_16_master
  bone2Name=clavicle_r
  swing1LimitDegrees=45
  swing2LimitDegrees=45
  swing2Motion=Locked
  twistLimitDegrees=45
  twistMotion=Locked
 }
 {
  bone1Name=dyn_r_shoulder_leaf_15_master
  bone2Name=clavicle_r
  swing1LimitDegrees=45
  swing2LimitDegrees=45
  swing2Motion=Locked
  twistLimitDegrees=45
  twistMotion=Locked
 }
 {
  bone1Name=dyn_r_shoulder_leaf_14_master
  bone2Name=clavicle_r
  swing1LimitDegrees=45
  swing2LimitDegrees=45
  swing2Motion=Locked
  twistLimitDegrees=45
  twistMotion=Locked
 }
 {
  bone1Name=dyn_r_shoulder_leaf_13_master
  bone2Name=clavicle_r
  swing1LimitDegrees=45
  swing2LimitDegrees=45
  swing2Motion=Locked
  twistLimitDegrees=45
  twistMotion=Locked
 }
 {
  bone1Name=dyn_r_shoulder_leaf_12_master
  bone2Name=clavicle_r
  swing1LimitDegrees=45
  swing2LimitDegrees=45
  swing2Motion=Locked
  twistLimitDegrees=45
  twistMotion=Locked
 }
 {
  bone1Name=dyn_r_shoulder_leaf_11_master
  bone2Name=clavicle_r
  swing1LimitDegrees=45
  swing2LimitDegrees=45
  swing2Motion=Locked
  twistLimitDegrees=45
  twistMotion=Locked
 }
 {
  bone1Name=dyn_r_shoulder_leaf_10_master
  bone2Name=clavicle_r
  swing1LimitDegrees=45
  swing2LimitDegrees=45
  swing2Motion=Locked
  twistLimitDegrees=45
  twistMotion=Locked
 }
 {
  bone1Name=dyn_r_shoulder_leaf_09_master
  bone2Name=clavicle_r
  swing1LimitDegrees=45
  swing2LimitDegrees=45
  swing2Motion=Locked
  twistLimitDegrees=45
  twistMotion=Locked
 }
 {
  bone1Name=dyn_r_shoulder_leaf_08_master
  bone2Name=clavicle_r
  swing1LimitDegrees=45
  swing2LimitDegrees=45
  swing2Motion=Locked
  twistLimitDegrees=45
  twistMotion=Locked
 }
 {
  bone1Name=dyn_r_shoulder_leaf_07_master
  bone2Name=clavicle_r
  swing1LimitDegrees=45
  swing2LimitDegrees=45
  swing2Motion=Locked
  twistLimitDegrees=45
  twistMotion=Locked
 }
 {
  bone1Name=dyn_r_shoulder_leaf_06_master
  bone2Name=clavicle_r
  swing1LimitDegrees=45
  swing2LimitDegrees=45
  swing2Motion=Locked
  twistLimitDegrees=45
  twistMotion=Locked
 }
 {
  bone1Name=dyn_r_shoulder_leaf_05_master
  bone2Name=clavicle_r
  swing1LimitDegrees=45
  swing2LimitDegrees=45
  swing2Motion=Locked
  twistLimitDegrees=45
  twistMotion=Locked
 }
 {
  bone1Name=dyn_r_shoulder_leaf_04_master
  bone2Name=clavicle_r
  swing1LimitDegrees=45
  swing2LimitDegrees=45
  swing2Motion=Locked
  twistLimitDegrees=45
  twistMotion=Locked
 }
 {
  bone1Name=dyn_r_shoulder_leaf_03_master
  bone2Name=clavicle_r
  swing1LimitDegrees=45
  swing2LimitDegrees=45
  swing2Motion=Locked
  twistLimitDegrees=45
  twistMotion=Locked
 }
 {
  bone1Name=dyn_r_shoulder_leaf_02_master
  bone2Name=clavicle_r
  swing1LimitDegrees=45
  swing2LimitDegrees=45
  swing2Motion=Locked
  twistLimitDegrees=45
  twistMotion=Locked
 }
 {
  bone1Name=dyn_r_shoulder_leaf_01_master
  bone2Name=clavicle_r
  swing1LimitDegrees=45
  swing2LimitDegrees=45
  swing2Motion=Locked
  twistLimitDegrees=45
  twistMotion=Locked
 }
 {
  bone1Name=dyn_r_shoulder_flower_05_master
  bone2Name=clavicle_r
  swing1LimitDegrees=45
  swing2LimitDegrees=45
  swing2Motion=Locked
  twistLimitDegrees=45
  twistMotion=Locked
 }
 {
  bone1Name=dyn_r_shoulder_flower_04_master
  bone2Name=clavicle_r
  swing1LimitDegrees=45
  swing2LimitDegrees=45
  swing2Motion=Locked
  twistLimitDegrees=45
  twistMotion=Locked
 }
 {
  bone1Name=dyn_r_shoulder_flower_03_master
  bone2Name=clavicle_r
  swing1LimitDegrees=45
  swing2LimitDegrees=45
  swing2Motion=Locked
  twistLimitDegrees=45
  twistMotion=Locked
 }
 {
  bone1Name=dyn_r_shoulder_flower_02_master
  bone2Name=clavicle_r
  swing1LimitDegrees=45
  swing2LimitDegrees=45
  swing2Motion=Locked
  twistLimitDegrees=45
  twistMotion=Locked
 }
 {
  bone1Name=dyn_r_shoulder_flower_01_master
  bone2Name=clavicle_r
  swing1LimitDegrees=45
  swing2LimitDegrees=45
  swing2Motion=Locked
  twistLimitDegrees=45
  twistMotion=Locked
 }
 {
  bone1Name=dyn_l_shoulder_leaf_01_master
  bone2Name=clavicle_l
  swing1LimitDegrees=45
  swing2LimitDegrees=45
  swing2Motion=Locked
  twistLimitDegrees=45
  twistMotion=Locked
 }
 {
  bone1Name=dyn_l_shoulder_leaf_02_master
  bone2Name=clavicle_l
  swing1LimitDegrees=45
  swing2LimitDegrees=45
  swing2Motion=Locked
  twistLimitDegrees=45
  twistMotion=Locked
 }
 {
  bone1Name=dyn_l_shoulder_leaf_03_master
  bone2Name=clavicle_l
  swing1LimitDegrees=45
  swing2LimitDegrees=45
  swing2Motion=Locked
  twistLimitDegrees=45
  twistMotion=Locked
 }
 {
  bone1Name=dyn_l_shoulder_leaf_04_master
  bone2Name=clavicle_l
  swing1LimitDegrees=45
  swing2LimitDegrees=45
  swing2Motion=Locked
  twistLimitDegrees=45
  twistMotion=Locked
 }
 {
  bone1Name=dyn_l_shoulder_leaf_05_master
  bone2Name=clavicle_l
  swing1LimitDegrees=45
  swing2LimitDegrees=45
  swing2Motion=Locked
  twistLimitDegrees=45
  twistMotion=Locked
 }
 {
  bone1Name=dyn_l_shoulder_leaf_06_master
  bone2Name=clavicle_l
  swing1LimitDegrees=45
  swing2LimitDegrees=45
  swing2Motion=Locked
  twistLimitDegrees=45
  twistMotion=Locked
 }
 {
  bone1Name=dyn_l_shoulder_leaf_07_master
  bone2Name=clavicle_l
  swing1LimitDegrees=45
  swing2LimitDegrees=45
  swing2Motion=Locked
  twistLimitDegrees=45
  twistMotion=Locked
 }
 {
  bone1Name=dyn_l_shoulder_leaf_08_master
  bone2Name=clavicle_l
  swing1LimitDegrees=45
  swing2LimitDegrees=45
  swing2Motion=Locked
  twistLimitDegrees=45
  twistMotion=Locked
 }
 {
  bone1Name=dyn_l_shoulder_leaf_09_master
  bone2Name=clavicle_l
  swing1LimitDegrees=45
  swing2LimitDegrees=45
  swing2Motion=Locked
  twistLimitDegrees=45
  twistMotion=Locked
 }
 {
  bone1Name=dyn_l_shoulder_leaf_010_master
  bone2Name=clavicle_l
  swing1LimitDegrees=45
  swing2LimitDegrees=45
  swing2Motion=Locked
  twistLimitDegrees=45
  twistMotion=Locked
 }
 {
  bone1Name=dyn_l_shoulder_leaf_011_master
  bone2Name=clavicle_l
  swing1LimitDegrees=45
  swing2LimitDegrees=45
  swing2Motion=Locked
  twistLimitDegrees=45
  twistMotion=Locked
 }
 {
  bone1Name=dyn_l_shoulder_leaf_012_master
  bone2Name=clavicle_l
  swing1LimitDegrees=45
  swing2LimitDegrees=45
  swing2Motion=Locked
  twistLimitDegrees=45
  twistMotion=Locked
 }
 {
  bone1Name=dyn_l_shoulder_leaf_013_master
  bone2Name=clavicle_l
  swing1LimitDegrees=45
  swing2LimitDegrees=45
  swing2Motion=Locked
  twistLimitDegrees=45
  twistMotion=Locked
 }
 {
  bone1Name=dyn_l_shoulder_leaf_014_master
  bone2Name=clavicle_l
  swing1LimitDegrees=45
  swing2LimitDegrees=45
  swing2Motion=Locked
  twistLimitDegrees=45
  twistMotion=Locked
 }
 {
  bone1Name=dyn_l_arm_flower_03_master
  bone2Name=clavicle_l
  swing1LimitDegrees=45
  swing2LimitDegrees=45
  swing2Motion=Locked
  twistLimitDegrees=45
  twistMotion=Locked
 }
 {
  bone1Name=dyn_l_arm_flower_02_master
  bone2Name=clavicle_l
  swing1LimitDegrees=45
  swing2LimitDegrees=45
  swing2Motion=Locked
  twistLimitDegrees=45
  twistMotion=Locked
 }
 {
  bone1Name=dyn_l_shoulder_flower_01_master
  bone2Name=clavicle_l
  swing1LimitDegrees=45
  swing2LimitDegrees=45
  swing2Motion=Locked
  twistLimitDegrees=45
  twistMotion=Locked
 }
 {
  bone1Name=dyn_l_shoulder_flower_02_master
  bone2Name=clavicle_l
  swing1LimitDegrees=45
  swing2LimitDegrees=45
  swing2Motion=Locked
  twistLimitDegrees=45
  twistMotion=Locked
 }
 {
  bone1Name=dyn_l_shoulder_flower_03_master
  bone2Name=clavicle_l
  swing1LimitDegrees=45
  swing2LimitDegrees=45
  swing2Motion=Locked
  twistLimitDegrees=45
  twistMotion=Locked
 }
 {
  bone1Name=dyn_l_shoulder_flower_04_master
  bone2Name=clavicle_l
  swing1LimitDegrees=45
  swing2LimitDegrees=45
  swing2Motion=Locked
  twistLimitDegrees=45
  twistMotion=Locked
 }
 {
  bone1Name=dyn_l_shoulder_flower_06_master
  bone2Name=clavicle_l
  swing1LimitDegrees=45
  swing2LimitDegrees=45
  swing2Motion=Locked
  twistLimitDegrees=45
  twistMotion=Locked
 }
 {
  bone1Name=dyn_l_shoulder_flower_08_master
  bone2Name=clavicle_l
  swing1LimitDegrees=45
  swing2LimitDegrees=45
  swing2Motion=Locked
  twistLimitDegrees=45
  twistMotion=Locked
 }
 {
  bone1Name=dyn_l_shoulder_flower_07_master
  bone2Name=clavicle_l
  swing1LimitDegrees=45
  swing2LimitDegrees=45
  swing2Motion=Locked
  twistLimitDegrees=45
  twistMotion=Locked
 }
 {
  bone1Name=dyn_l_arm_flower_01_master
  bone2Name=clavicle_l
  swing1LimitDegrees=45
  swing2LimitDegrees=45
  swing2Motion=Locked
  twistLimitDegrees=45
  twistMotion=Locked
 }
 {
  bone1Name=dyn_r_hand_flower_01_master
  bone2Name=clavicle_r
  swing1LimitDegrees=45
  swing2LimitDegrees=45
  swing2Motion=Locked
  twistLimitDegrees=45
  twistMotion=Locked
 }
 {
  bone1Name=dyn_r_hand_flower_02_master
  bone2Name=clavicle_r
  swing1LimitDegrees=45
  swing2LimitDegrees=45
  swing2Motion=Locked
  twistLimitDegrees=45
  twistMotion=Locked
 }
 {
  bone1Name=dyn_r_arm_leaf_04_master
  bone2Name=clavicle_r
  swing1LimitDegrees=45
  swing2LimitDegrees=45
  swing2Motion=Locked
  twistLimitDegrees=45
  twistMotion=Locked
 }
 {
  bone1Name=dyn_r_arm_leaf_02_master
  bone2Name=clavicle_r
  swing1LimitDegrees=45
  swing2LimitDegrees=45
  swing2Motion=Locked
  twistLimitDegrees=45
  twistMotion=Locked
 }
 {
  bone1Name=dyn_r_arm_leaf_03_master
  bone2Name=clavicle_r
  swing1LimitDegrees=45
  swing2LimitDegrees=45
  swing2Motion=Locked
  twistLimitDegrees=45
  twistMotion=Locked
 }
 {
  bone1Name=dyn_r_arm_leaf_01_master
  bone2Name=clavicle_r
  swing1LimitDegrees=45
  swing2LimitDegrees=45
  swing2Motion=Locked
  twistLimitDegrees=45
  twistMotion=Locked
 }
 {
  bone1Name=dyn_l_shoulder_flower_09_master
  bone2Name=clavicle_l
  swing1LimitDegrees=45
  swing2LimitDegrees=45
  swing2Motion=Locked
  twistLimitDegrees=45
  twistMotion=Locked
 }
 {
  bone1Name=dyn_l_shoulder_flower_05_master
  bone2Name=clavicle_l
  swing1LimitDegrees=45
  swing2LimitDegrees=45
  swing2Motion=Locked
  twistLimitDegrees=45
  twistMotion=Locked
 }
 {
  bone1Name=dyn_r_shoulder_leaf_19
  bone2Name=dyn_r_shoulder_leaf_19_master
  swing1LimitDegrees=15
  swing2LimitDegrees=20
  twistLimitDegrees=10
 }
 {
  bone1Name=dyn_r_shoulder_leaf_19_master
  bone2Name=clavicle_r
  swing1LimitDegrees=45
  swing1Motion=Free
  swing2LimitDegrees=45
  swing2Motion=Free
  twistLimitDegrees=45
  twistMotion=Free
 }
 {
  bone1Name=dyn_l_shoulder_leaf_015_master
  bone2Name=clavicle_l
  swing1LimitDegrees=45
  swing1Motion=Free
  swing2LimitDegrees=45
  swing2Motion=Free
  twistLimitDegrees=45
  twistMotion=Free
 }
 {
  bone1Name=dyn_l_shoulder_leaf_015
  bone2Name=dyn_l_shoulder_leaf_015_master
  swing1LimitDegrees=10
  swing2LimitDegrees=10
  twistLimitDegrees=10
 }
]
