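# Physics asset for the Moss Knight medium male player cosmetic body, rigging its dangling vines, leaves, fern, foot flower and bandana for physics.
# Capsules cover a partial skeleton (pelvis, spine, left arm and thigh, calves, feet) to which the dyn_ chains attach, many through small master bodies.
# Chains use narrow swing/twist limits, while limb joints and most masters rotate freely.
# The bandana master, leg-vine root and wobble bodies are angularly locked.
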PhysicsAsset /BRCosmetics/Characters/Player/Male/Medium/Bodies/M_MED_Moss_Knight/Meshes/M_MED_Moss_Knight_Physics
bodies {
 calf_l [{shapeType=Capsule, length=28.7426, radius=6.6853, center=<-22.2562 0 0.0003>, rotation=<rot 90 0 0>}]
 calf_r [{shapeType=Capsule, length=25.94, radius=5.9925, center=<25.8703 0.0001 0>, rotation=<rot 90 0 0>}]
 dyn_back_vine_l_01 [{shapeType=Capsule, length=3.438, radius=0.75, center=<2.6361 0 0>, rotation=<rot 90 0 0>}]
 dyn_back_vine_l_02 [{shapeType=Capsule, length=5.5515, radius=0.75, center=<3.4556 0 0.0001>, rotation=<rot 90 0 0>}]
 dyn_back_vine_l_03 [{shapeType=Capsule, length=8.624, radius=0.75, center=<5.1464 0.0002 0>, rotation=<rot 90 0 0>}]
 dyn_back_vine_l_04 [{shapeType=Capsule, length=5.5951, radius=0.75, center=<3.7737 0 0>, rotation=<rot 90 0 0>}]
 dyn_back_vine_l_05 [{shapeType=Capsule, length=3.6047, radius=0.75, center=<2.6361 0 0>, rotation=<rot 90 0 0>}]
 dyn_back_vine_l_06 [{shapeType=Capsule, length=2.3225, radius=0.75, center=<1.9312 0 0>, rotation=<rot 90 0 0>}]
 dyn_back_vine_l_07 [{shapeType=Capsule, length=3.4936, radius=0.75, center=<2.6361 0 0>, rotation=<rot 90 0 0>}]
 dyn_back_vine_l_08 [{shapeType=Capsule, length=4.0199, radius=0.75, center=<3.4319 0 0>, rotation=<rot 90 0 0>}]
 dyn_back_vine_l_leaf_01 [{shapeType=Capsule, length=2.0617, radius=0.505, center=<1.5323 0.0111 -0.0339>, rotation=<rot 0 97.2609 -83.1849>}]
 dyn_back_vine_l_leaf_01_b [{shapeType=Capsule, length=2.7168, radius=1.8514, center=<2.5606 -0.3025 0.4307>, rotation=<rot 0 82.7226 -87.5917>}]
 dyn_back_vine_l_leaf_02 [{shapeType=Capsule, length=2.8498, radius=1.8797, center=<2.8513 0.0868 0.2471>, rotation=<rot 0 88.8347 -88.3864>}]
 dyn_back_vine_l_leaf_03 [{shapeType=Capsule, length=3.1927, radius=1.8282, center=<3.7184 -0.0064 0.2924>, rotation=<rot 0 88.9475 -88.2435>}]
 dyn_back_vine_l_leaf_04 [{shapeType=Capsule, length=2.1669, radius=1.2222, center=<2.3228 0.0045 -0.032>, rotation=<rot 0 91.2482 -89.1817>}]
 dyn_back_vine_l_leaf_05 [{shapeType=Capsule, length=2.6391, radius=1.3156, center=<2.9689 0.4703 -0.2017>, rotation=<rot 0 -93.4837 -83.2523>}]
 dyn_back_vine_l_master [{shapeType=Capsule, length=0.505, radius=0.505, rotation=<rot 0 -90 0>}]
 dyn_l_back_shoulder_fern_01 [{shapeType=Capsule, length=1.7908, radius=0.5, center=<1.4818 0 0>, rotation=<rot 90 0 0>}]
 dyn_l_back_shoulder_fern_02 [{shapeType=Capsule, length=1.568, radius=0.7228, center=<1.4818 0 0>, rotation=<rot 90 0 0>}]
 dyn_l_back_shoulder_fern_03 [{shapeType=Capsule, length=1.9175, radius=0.7286, center=<1.6536 0 0>, rotation=<rot 90 0 0>}]
 dyn_l_back_shoulder_fern_leaf_01 [{shapeType=Capsule, length=2.2342, radius=1.4733, center=<2.6705 0 0>, rotation=<rot 90 0 0>}]
 dyn_l_back_shoulder_fern_leaf_02 [{shapeType=Capsule, length=2.0061, radius=1.0782, center=<2.12 0 0>, rotation=<rot 90 0 0>}]
 dyn_l_back_shoulder_fern_leaf_03 [{shapeType=Capsule, length=2.9327, radius=1.7602, center=<2.6705 0 0>, rotation=<rot 90 0 0>}]
 dyn_l_back_shoulder_fern_master [{shapeType=Capsule, length=0.505, radius=0.505, rotation=<rot 0 -90 0>}]
 dyn_l_foot_leaf_01 [{shapeType=Capsule, length=4.467, radius=1.1002, center=<3.5194 -0.1843 -0.1832>, rotation=<rot 0 91.6634 -89.2847>}]
 dyn_l_foot_leaf_02 [{shapeType=Capsule, length=3.1339, radius=1.0476, center=<2.5314 0.0033 0.2721>, rotation=<rot 0 91.9394 -78.2797>}]
 dyn_l_leg_leaf_01 [{shapeType=Capsule, length=1.8774, radius=1.0255, center=<1.8378 0.0745 0.0221>, rotation=<rot 0 90.9809 -89.2928>}]
 dyn_l_leg_leaf_01_master [{shapeType=Capsule, length=0.505, radius=0.505, rotation=<rot 0 -90 0>}]
 dyn_l_leg_leaf_02 [{shapeType=Capsule, length=3.327, radius=1.9225, center=<1.6931 0.0118 0.1406>, rotation=<rot 0 79.0987 -88.9694>}]
 dyn_l_leg_leaf_02_master [{shapeType=Capsule, length=0.505, radius=0.505, rotation=<rot 0 -90 0>}]
 dyn_l_leg_leaf_03 [{shapeType=Capsule, length=3.4603, radius=1.6595, center=<2.4338 0.0266 0.1089>, rotation=<rot 0 88.9726 -87.4126>}]
 dyn_l_leg_leaf_03_master [{shapeType=Capsule, length=0.505, radius=0.505, rotation=<rot 0 -90 0>}]
 dyn_l_leg_leaf_04 [{shapeType=Capsule, length=1.798, radius=0.5299, center=<1.9082 -0.0696 0.034>, rotation=<rot 0 90.5237 -84.2365>}]
 dyn_l_leg_leaf_04_b [{shapeType=Capsule, length=2.863, radius=1.9013, center=<2.8419 0.0133 0.1198>, rotation=<rot 0 90.6512 -86.8022>}]
 dyn_l_leg_leaf_04_master [{shapeType=Capsule, length=0.505, radius=0.505, rotation=<rot 0 -90 0>}]
 dyn_l_leg_leaf_05 [{shapeType=Capsule, length=3.8531, radius=2.6145, center=<1.248 -2.1934 -0.3236>, rotation=<rot 0 155.9861 -83.4851>}]
 dyn_l_leg_leaf_05_master [{shapeType=Capsule, length=0.505, radius=0.505, rotation=<rot 0 -90 0>}]
 dyn_l_leg_leaf_06 [{shapeType=Capsule, length=4.3653, radius=1.6321, center=<3.1289 0 0.0001>, rotation=<rot 90 0 0>}]
 dyn_l_leg_leaf_06_master [{shapeType=Capsule, length=0.505, radius=0.505, rotation=<rot 0 -90 0>}]
 dyn_l_leg_vine_wobble_01 [{shapeType=Capsule, length=0.505, radius=0.505, center=<0 0 -0.4196>, rotation=<rot 90 0 0>}]
 dyn_l_leg_vine_wobble_02 [{shapeType=Capsule, length=0.505, radius=0.505, center=<0 0 -0.641>, rotation=<rot 0 0 90>}]
 dyn_leg_vine_top_01 [{shapeType=Capsule, length=4.8351, radius=0.75, center=<3.2525 0 -0.0003>, rotation=<rot 90 0 0>}]
 dyn_leg_vine_top_02 [{shapeType=Capsule, length=5.4593, radius=0.75, center=<3.6866 0 -0.0002>, rotation=<rot 90 0 0>}]
 dyn_leg_vine_top_03 [{shapeType=Capsule, length=5.4593, radius=0.75, center=<3.6866 0 -0.0002>, rotation=<rot 90 0 0>}]
 dyn_leg_vine_top_04 [{shapeType=Capsule, length=2.9427, radius=0.75, center=<2.2668 0 -0.0001>, rotation=<rot 90 0 0>}]
 dyn_leg_vine_top_05 [{shapeType=Capsule, length=2.3813, radius=1.4271, center=<2.5652 -0.0743 0.4361>, rotation=<rot 0 -88.9188 -86.1696>}]
 dyn_leg_vine_top_leaf_01 [{shapeType=Capsule, length=2.8073, radius=1.8238, center=<3.062 -0.0531 0.3328>, rotation=<rot 0 89.4499 -88.5279>}]
 dyn_leg_vine_top_leaf_02 [{shapeType=Capsule, length=2.2358, radius=1.4869, center=<2.2536 -0.041 0.04>, rotation=<rot 0 87.6754 -87.4334>}]
 dyn_leg_vine_top_leaf_03 [{shapeType=Capsule, length=1.9827, radius=1.098, center=<1.9647 0.1055 0.3462>, rotation=<rot 0 -89.9804 -86.4425>}]
 dyn_r_bandana_bot_01 [{shapeType=Capsule, length=0.8174, radius=2.1369, center=<1.8626 0 -0.0001>, rotation=<rot 90 0 0>}]
 dyn_r_bandana_bot_02 [{shapeType=Capsule, length=2.8063, radius=1.7202, center=<2.8588 -0.0603 0.3156>, rotation=<rot 0 83.495 -87.3813>}]
 dyn_r_bandana_master [{shapeType=Capsule, length=0, radius=1.5491}]
 dyn_r_bandana_top_01 [{shapeType=Capsule, length=0.9033, radius=1.6756, center=<1.8626 0 -0.0001>, rotation=<rot 90 0 0>}]
 dyn_r_bandana_top_02 [{shapeType=Capsule, length=2.911, radius=1.3097, center=<3.104 0.1077 -0.1799>, rotation=<rot 0 -93.6717 -85.0254>}]
 dyn_r_foot_flower_01 [{shapeType=Capsule, length=2.0349, radius=2.0269, center=<0.6214 0.0527 0.0291>, rotation=<rot -6.8726 -90 3.8612>}]
 dyn_r_foot_leaf_01 [{shapeType=Capsule, length=3.2725, radius=1.4898, center=<3.2724 0.0001 0>, rotation=<rot 90 0 0>}]
 dyn_r_foot_leaf_02 [
  {shapeType=Capsule, length=7.2251, radius=0.739, center=<4.7607 0.0001 0>, rotation=<rot 90 0 0>}
  {shapeType=Capsule, length=2.1517, radius=2.338, center=<12.7408 0.4922 2.783>, rotation=<rot 56.9454 -179.9999 179.9999>}
 ]
 dyn_r_foot_leaf_03 [{shapeType=Capsule, length=2.5102, radius=1.4898, center=<2.9792 0.0001 0>, rotation=<rot 90 0 0>}]
 dyn_r_foot_leaf_04 [{shapeType=Capsule, length=5.8461, radius=0.6276, center=<3.8648 0.0001 0>, rotation=<rot 90 0 0>}]
 dyn_r_foot_leaf_04_b [
  {shapeType=Capsule, length=6.1413, radius=1.4898, center=<4.767 0.0002 0>, rotation=<rot 90 0 0>}
  {shapeType=Capsule, length=0.9713, radius=2.8448, center=<9.5742 0.0002 -0.0001>, rotation=<rot 90 0 0>}
 ]
 dyn_r_foot_leaf_05 [{shapeType=Capsule, length=3.5, radius=0.5, center=<3.3341 0.0001 0>, rotation=<rot 90 0 0>}]
 dyn_r_foot_leaf_06 [{shapeType=Capsule, length=2.5102, radius=1.4898, center=<2.7222 0.0001 0>, rotation=<rot 90 0 0>}]
 foot_l [{shapeType=Capsule, length=9.565, radius=8.0531, center=<-0.9999 8.606 0>, rotation=<rot 0 0 90>}]
 foot_r [{shapeType=Capsule, length=14.032, radius=5.6305, center=<1.8009 -7.459 0>, rotation=<rot 0 0 90>}]
 lowerarm_l [{shapeType=Capsule, length=17.7994, radius=5.4209, center=<12.7915 0.0006 0>, rotation=<rot 90 0 0>}]
 pelvis [{shapeType=Capsule, length=10.5378, radius=16.2506}]
 spine_01 [{shapeType=Capsule, length=9.0667, radius=16.3255}]
 spine_03 [{shapeType=Capsule, length=8.6896, radius=16.1664}]
 spine_05 [{shapeType=Capsule, length=13.5797, radius=14.8657}]
 thigh_l [{shapeType=Capsule, length=19.0451, radius=13.0286, center=<-20.961 0.0005 0>, rotation=<rot 90 0 0>}]
 upperarm_l [{shapeType=Capsule, length=17.7994, radius=5.4209, center=<14.6909 0.0008 0>, rotation=<rot 90 0 0>}]
}
constraints [
 {bone1Name=spine_01, bone2Name=pelvis, swing1LimitDegrees=45, swing2LimitDegrees=45, twistLimitDegrees=45}
 {bone1Name=spine_03, bone2Name=spine_01, swing1LimitDegrees=45, swing2LimitDegrees=45, twistLimitDegrees=45}
 {bone1Name=spine_05, bone2Name=spine_03, swing1LimitDegrees=45, swing2LimitDegrees=45, twistLimitDegrees=45}
 {bone1Name=dyn_back_vine_l_01, bone2Name=dyn_back_vine_l_master, swing1LimitDegrees=10, swing2LimitDegrees=10, twistLimitDegrees=10}
 {bone1Name=dyn_back_vine_l_02, bone2Name=dyn_back_vine_l_01, swing1LimitDegrees=30, swing2LimitDegrees=30, twistLimitDegrees=30}
 {bone1Name=dyn_back_vine_l_03, bone2Name=dyn_back_vine_l_02, swing1LimitDegrees=30, swing2LimitDegrees=30, twistLimitDegrees=30}
 {bone1Name=dyn_back_vine_l_04, bone2Name=dyn_back_vine_l_03, swing1LimitDegrees=30, swing2LimitDegrees=30, twistLimitDegrees=30}
 {bone1Name=dyn_back_vine_l_05, bone2Name=dyn_back_vine_l_04, swing1LimitDegrees=30, swing2LimitDegrees=30, twistLimitDegrees=30}
 {bone1Name=dyn_back_vine_l_06, bone2Name=dyn_back_vine_l_05, swing1LimitDegrees=30, swing2LimitDegrees=30, twistLimitDegrees=30}
 {bone1Name=dyn_back_vine_l_07, bone2Name=dyn_back_vine_l_06, swing1LimitDegrees=30, swing2LimitDegrees=30, twistLimitDegrees=30}
 {bone1Name=dyn_back_vine_l_08, bone2Name=dyn_back_vine_l_07, swing1LimitDegrees=30, swing2LimitDegrees=30, twistLimitDegrees=30}
 {bone1Name=dyn_back_vine_l_leaf_01_b, bone2Name=dyn_back_vine_l_leaf_01, swing1LimitDegrees=20, swing2LimitDegrees=25, twistLimitDegrees=5}
 {bone1Name=dyn_back_vine_l_leaf_01, bone2Name=dyn_back_vine_l_04, swing1LimitDegrees=30, swing2LimitDegrees=30, twistLimitDegrees=30}
 {bone1Name=dyn_back_vine_l_leaf_02, bone2Name=dyn_back_vine_l_05, swing1LimitDegrees=20, swing2LimitDegrees=25, twistLimitDegrees=5}
 {bone1Name=dyn_back_vine_l_leaf_04, bone2Name=dyn_back_vine_l_07, swing1LimitDegrees=20, swing2LimitDegrees=25, twistLimitDegrees=5}
 {bone1Name=dyn_back_vine_l_leaf_03, bone2Name=dyn_back_vine_l_07, swing1LimitDegrees=20, swing2LimitDegrees=25, twistLimitDegrees=5}
 {bone1Name=dyn_back_vine_l_leaf_05, bone2Name=dyn_back_vine_l_08, swing1LimitDegrees=20, swing2LimitDegrees=25, twistLimitDegrees=5}
 {
  bone1Name=thigh_l
  bone2Name=pelvis
  swing1LimitDegrees=45
  swing1Motion=Free
  swing2LimitDegrees=45
  swing2Motion=Free
  twistLimitDegrees=45
  twistMotion=Free
 }
 {
  bone1Name=calf_l
  bone2Name=thigh_l
  swing1LimitDegrees=45
  swing1Motion=Free
  swing2LimitDegrees=45
  swing2Motion=Free
  twistLimitDegrees=45
  twistMotion=Free
 }
 {
  bone1Name=dyn_leg_vine_top_01
  bone2Name=thigh_l
  swing1LimitDegrees=7
  swing1Motion=Locked
  swing2LimitDegrees=7
  swing2Motion=Locked
  twistLimitDegrees=7
  twistMotion=Locked
 }
 {bone1Name=dyn_leg_vine_top_02, bone2Name=dyn_leg_vine_top_01, swing1LimitDegrees=10, swing2LimitDegrees=10, twistLimitDegrees=10}
 {bone1Name=dyn_leg_vine_top_03, bone2Name=dyn_leg_vine_top_02, swing1LimitDegrees=15, swing2LimitDegrees=15, twistLimitDegrees=15}
 {bone1Name=dyn_leg_vine_top_04, bone2Name=dyn_leg_vine_top_03, swing1LimitDegrees=15, swing2LimitDegrees=15, twistLimitDegrees=15}
 {bone1Name=dyn_leg_vine_top_05, bone2Name=dyn_leg_vine_top_04, swing1LimitDegrees=15, swing2LimitDegrees=25, twistLimitDegrees=5}
 {bone1Name=dyn_leg_vine_top_leaf_03, bone2Name=dyn_leg_vine_top_03, swing1LimitDegrees=25, swing2LimitDegrees=45, twistLimitDegrees=5}
 {bone1Name=dyn_leg_vine_top_leaf_02, bone2Name=dyn_leg_vine_top_03, swing1LimitDegrees=25, swing2LimitDegrees=45, twistLimitDegrees=5}
 {bone1Name=dyn_leg_vine_top_leaf_01, bone2Name=dyn_leg_vine_top_01, swing1LimitDegrees=25, swing2LimitDegrees=45, twistLimitDegrees=5}
 {
  bone1Name=dyn_l_leg_vine_wobble_01
  bone2Name=thigh_l
  swing1LimitDegrees=5
  swing1Motion=Locked
  swing2LimitDegrees=5
  swing2Motion=Locked
  twistLimitDegrees=5
  twistMotion=Locked
 }
 {
  bone1Name=dyn_l_leg_vine_wobble_02
  bone2Name=calf_l
  swing1LimitDegrees=5
  swing1Motion=Locked
  swing2LimitDegrees=5
  swing2Motion=Locked
  twistLimitDegrees=5
  twistMotion=Locked
 }
 {bone1Name=dyn_l_leg_leaf_05, bone2Name=dyn_l_leg_leaf_05_master, swing1LimitDegrees=15, swing2LimitDegrees=30, twistLimitDegrees=5}
 {bone1Name=dyn_l_leg_leaf_06, bone2Name=dyn_l_leg_leaf_06_master, swing1LimitDegrees=15, swing2LimitDegrees=30, twistLimitDegrees=5}
 {bone1Name=dyn_l_leg_leaf_01, bone2Name=dyn_l_leg_leaf_01_master, swing1LimitDegrees=15, swing2LimitDegrees=30, twistLimitDegrees=5}
 {bone1Name=dyn_l_leg_leaf_02, bone2Name=dyn_l_leg_leaf_02_master, swing1LimitDegrees=15, swing2LimitDegrees=30, twistLimitDegrees=5}
 {bone1Name=dyn_l_leg_leaf_03, bone2Name=dyn_l_leg_leaf_03_master, swing1LimitDegrees=15, swing2LimitDegrees=30, twistLimitDegrees=5}
 {bone1Name=dyn_l_leg_leaf_04, bone2Name=dyn_l_leg_leaf_04_master, swing1LimitDegrees=15, swing2LimitDegrees=30, twistLimitDegrees=5}
 {bone1Name=dyn_l_leg_leaf_04_b, bone2Name=dyn_l_leg_leaf_04, swing1LimitDegrees=15, swing2LimitDegrees=30, twistLimitDegrees=5}
 {bone1Name=thigh_l, bone2Name=dyn_leg_vine_top_03, swing1LimitDegrees=45, swing2LimitDegrees=45, twistLimitDegrees=45}
 {
  bone1Name=dyn_l_back_shoulder_fern_01
  bone2Name=dyn_l_back_shoulder_fern_master
  swing1LimitDegrees=5
  swing2LimitDegrees=5
  twistLimitDegrees=5
 }
 {
  bone1Name=dyn_l_back_shoulder_fern_02
  bone2Name=dyn_l_back_shoulder_fern_01
  swing1LimitDegrees=15
  swing2LimitDegrees=15
  twistLimitDegrees=15
 }
 {
  bone1Name=dyn_l_back_shoulder_fern_03
  bone2Name=dyn_l_back_shoulder_fern_02
  swing1LimitDegrees=15
  swing2LimitDegrees=15
  twistLimitDegrees=15
 }
 {
  bone1Name=dyn_l_back_shoulder_fern_leaf_03
  bone2Name=dyn_l_back_shoulder_fern_03
  swing1LimitDegrees=15
  swing2LimitDegrees=30
  twistLimitDegrees=5
 }
 {
  bone1Name=dyn_l_back_shoulder_fern_leaf_01
  bone2Name=dyn_l_back_shoulder_fern_02
  swing1LimitDegrees=25
  swing2LimitDegrees=45
  twistLimitDegrees=5
 }
 {
  bone1Name=dyn_l_back_shoulder_fern_leaf_02
  bone2Name=dyn_l_back_shoulder_fern_02
  swing1LimitDegrees=25
  swing2LimitDegrees=45
  twistLimitDegrees=5
 }
 {
  bone1Name=upperarm_l
  bone2Name=spine_05
  swing1LimitDegrees=45
  swing1Motion=Free
  swing2LimitDegrees=45
  swing2Motion=Free
  twistLimitDegrees=45
  twistMotion=Free
 }
 {
  bone1Name=lowerarm_l
  bone2Name=upperarm_l
  swing1LimitDegrees=45
  swing1Motion=Free
  swing2LimitDegrees=45
  swing2Motion=Free
  twistLimitDegrees=45
  twistMotion=Free
 }
 {
  bone1Name=dyn_r_bandana_master
  bone2Name=spine_05
  swing1LimitDegrees=45
  swing1Motion=Locked
  swing2LimitDegrees=45
  swing2Motion=Locked
  twistLimitDegrees=45
  twistMotion=Locked
 }
 {bone1Name=dyn_r_bandana_bot_01, bone2Name=dyn_r_bandana_master, swing1LimitDegrees=45, swing2LimitDegrees=45, twistLimitDegrees=5}
 {bone1Name=dyn_r_bandana_bot_02, bone2Name=dyn_r_bandana_bot_01, swing1LimitDegrees=45, swing2LimitDegrees=45, twistLimitDegrees=5}
 {bone1Name=dyn_r_bandana_top_01, bone2Name=dyn_r_bandana_master, swing1LimitDegrees=45, swing2LimitDegrees=45, twistLimitDegrees=5}
 {bone1Name=dyn_r_bandana_top_02, bone2Name=dyn_r_bandana_top_01, swing1LimitDegrees=45, swing2LimitDegrees=45, twistLimitDegrees=5}
 {
  bone1Name=foot_l
  bone2Name=calf_l
  swing1LimitDegrees=45
  swing1Motion=Free
  swing2LimitDegrees=45
  swing2Motion=Free
  twistLimitDegrees=45
  twistMotion=Free
 }
 {bone1Name=dyn_l_foot_leaf_01, bone2Name=foot_l, swing1LimitDegrees=10, swing2LimitDegrees=20, twistLimitDegrees=5}
 {bone1Name=dyn_l_foot_leaf_02, bone2Name=foot_l, swing1LimitDegrees=10, swing2LimitDegrees=20, twistLimitDegrees=5}
 {bone1Name=dyn_r_foot_leaf_01, bone2Name=foot_r, swing1LimitDegrees=10, swing2LimitDegrees=10, twistLimitDegrees=10}
 {bone1Name=dyn_r_foot_leaf_02, bone2Name=foot_r, swing1LimitDegrees=10, swing2LimitDegrees=10, twistLimitDegrees=10}
 {bone1Name=dyn_r_foot_leaf_03, bone2Name=foot_r, swing1LimitDegrees=10, swing2LimitDegrees=10, twistLimitDegrees=10}
 {bone1Name=dyn_r_foot_leaf_04, bone2Name=foot_r, swing1LimitDegrees=10, swing2LimitDegrees=10, twistLimitDegrees=10}
 {bone1Name=dyn_r_foot_leaf_05, bone2Name=foot_r, swing1LimitDegrees=10, swing2LimitDegrees=10, twistLimitDegrees=10}
 {bone1Name=dyn_r_foot_leaf_04_b, bone2Name=dyn_r_foot_leaf_04, swing1LimitDegrees=10, swing2LimitDegrees=10, twistLimitDegrees=10}
 {bone1Name=dyn_r_foot_flower_01, bone2Name=foot_r, swing1LimitDegrees=10, swing2LimitDegrees=10, twistLimitDegrees=10}
 {bone1Name=dyn_r_foot_leaf_06, bone2Name=foot_r, swing1LimitDegrees=20, swing2LimitDegrees=20, twistLimitDegrees=20}
 {
  bone1Name=foot_r
  bone2Name=calf_r
  swing1LimitDegrees=45
  swing1Motion=Free
  swing2LimitDegrees=45
  swing2Motion=Free
  twistLimitDegrees=45
  twistMotion=Free
 }
 {
  bone1Name=dyn_back_vine_l_master
  bone2Name=spine_05
  swing1LimitDegrees=45
  swing1Motion=Free
  swing2LimitDegrees=45
  swing2Motion=Free
  twistLimitDegrees=45
  twistMotion=Free
 }
 {
  bone1Name=dyn_l_back_shoulder_fern_master
  bone2Name=spine_05
  swing1LimitDegrees=45
  swing1Motion=Free
  swing2LimitDegrees=45
  swing2Motion=Free
  twistLimitDegrees=45
  twistMotion=Free
 }
 {
  bone1Name=calf_r
  bone2Name=pelvis
  swing1LimitDegrees=45
  swing1Motion=Free
  swing2LimitDegrees=45
  swing2Motion=Free
  twistLimitDegrees=45
  twistMotion=Free
 }
 {
  bone1Name=dyn_l_leg_leaf_05_master
  bone2Name=calf_l
  swing1LimitDegrees=45
  swing1Motion=Free
  swing2LimitDegrees=45
  swing2Motion=Free
  twistLimitDegrees=45
  twistMotion=Free
 }
 {
  bone1Name=dyn_l_leg_leaf_06_master
  bone2Name=calf_l
  swing1LimitDegrees=45
  swing1Motion=Free
  swing2LimitDegrees=45
  swing2Motion=Free
  twistLimitDegrees=45
  twistMotion=Free
 }
 {
  bone1Name=dyn_l_leg_leaf_01_master
  bone2Name=thigh_l
  swing1LimitDegrees=45
  swing1Motion=Free
  swing2LimitDegrees=45
  swing2Motion=Free
  twistLimitDegrees=45
  twistMotion=Free
 }
 {
  bone1Name=dyn_l_leg_leaf_02_master
  bone2Name=thigh_l
  swing1LimitDegrees=45
  swing1Motion=Free
  swing2LimitDegrees=45
  swing2Motion=Free
  twistLimitDegrees=45
  twistMotion=Free
 }
 {
  bone1Name=dyn_l_leg_leaf_03_master
  bone2Name=thigh_l
  swing1LimitDegrees=45
  swing1Motion=Free
  swing2LimitDegrees=45
  swing2Motion=Free
  twistLimitDegrees=45
  twistMotion=Free
 }
 {
  bone1Name=dyn_l_leg_leaf_04_master
  bone2Name=thigh_l
  swing1LimitDegrees=45
  swing1Motion=Free
  swing2LimitDegrees=45
  swing2Motion=Free
  twistLimitDegrees=45
  twistMotion=Free
 }
]
