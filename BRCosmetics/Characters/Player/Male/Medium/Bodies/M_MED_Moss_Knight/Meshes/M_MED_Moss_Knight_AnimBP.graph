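# Animation Blueprint for the Moss Knight body cosmetic, adding secondary physics motion to its dynamic bones, flowers and leaves.
# It copies the attached parent's pose and uses bIsInFrontEnd to choose between two stacked RigidBody simulations and a gameplay RigidBody with world-rotated dyn_root as simulation base.
# A long chain of spring-based AnimDynamics nodes then moves individual flower and leaf bones.
# ADS_Aim sets the alpha of the gameplay RigidBody and AnimDynamics nodes.
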AnimBlueprint /BRCosmetics/Characters/Player/Male/Medium/Bodies/M_MED_Moss_Knight/Meshes/M_MED_Moss_Knight_AnimBP
target_skeleton=&/M_MED_Moss_Knight_Skeleton
graphs [
 AnimGraph {nodes [
   n1 = "Transform (Modify) Bone - Bone: dyn_root"(
    ComponentPose=Local To Component(LocalPose=Copy Pose From Mesh() {bUseAttachedParent=true})
    Alpha=1
    Translation="0,0,0"
    Rotation=Get sim_space_rotation() -> (sim_space_rotation:Rotator)
    Scale="1,1,1"
   ) {
    rotationMode=BMM_Replace
    rotationSpace=BCS_WorldSpace
    translationSpace=BCS_BoneSpace
    alphaBoolBlend {blendOption=Linear}
    alphaScaleBiasClamp {bInterpResult=true, interpSpeedDecreasing=2, interpSpeedIncreasing=2}
    boneToModify {boneName=dyn_root}
   }
   n2 = RigidBody!(ComponentPose=@n1, Alpha=Get ADS_Aim() -> (ADS_Aim:double), OverrideWorldGravity:Vector="0,0,-980", ExternalForce) -> (Pose) {
    bClampLinearTranslationLimitToRefPose=true
    bOverrideWorldGravity=true
    overridePhysicsAsset=&/M_MED_Moss_Knight_Physics
    simulationSpace=BaseBoneSpace
    alphaBoolBlend {blendOption=Linear}
    baseBoneRef {boneName=dyn_root}
    componentAppliedLinearAccClamp=<1500 1500 1500>
    componentLinearAccScale=<0.3 0.3 0.4>
    componentLinearVelScale=<0.2 0.2 0.3>
    overrideWorldGravity {z=-1100}
   }
   n3 = Component To Local(ComponentPose=@n2)
   n4 = RigidBody(ComponentPose=Local To Component(LocalPose=Copy Pose From Mesh() {bUseAttachedParent=true}), OverrideWorldGravity="0,0,-980") {
    bClampLinearTranslationLimitToRefPose=true
    bOverrideWorldGravity=true
    overridePhysicsAsset=&M_MED_Moss_Knight_Physics
    simulationSpace=BaseBoneSpace
    alphaBoolBlend {blendOption=Linear}
    baseBoneRef {boneName=dyn_root}
    componentAppliedLinearAccClamp=<1500 1500 1500>
    componentLinearAccScale=<0.3 0.3 0.4>
    componentLinearVelScale=<0.2 0.2 0.3>
    overrideWorldGravity {z=-1100}
   }
   n5 = RigidBody(ComponentPose=@n4, OverrideWorldGravity="0,0,-980") {
    bClampLinearTranslationLimitToRefPose=true
    bOverrideWorldGravity=true
    overridePhysicsAsset=&/M_MED_Moss_Knight_Flower_Leaf_Physics
    simulationSpace=BaseBoneSpace
    alphaBoolBlend {blendOption=Linear}
    baseBoneRef {boneName=root}
    componentAppliedLinearAccClamp=<1500 1500 1500>
    componentLinearAccScale=<0.3 0.3 0.4>
    componentLinearVelScale=<0.2 0.2 0.3>
    overrideWorldGravity {z=-1100}
   }
   n6 = Component To Local(ComponentPose=@n5)
   n7 = Get bIsInFrontEnd()
   n8 = Get ADS_Aim()
   n9 = Anim Dynamics - Bone: dyn_r_shoulder_flower_01(
    ComponentPose=Local To Component(LocalPose=Blend Poses by bool(BlendPose_0=@n6, BlendPose_1=@n3, bActiveValue=@n7) {blendType=Linear})
    Alpha=@n8
    ExternalForce="0,0,0"
   ) {
    angularDampingOverride=200
    angularSpringConstant=500
    bAngularSpring=true
    bGravityOverrideInSimSpace=true
    bUseGravityOverride=true
    simulationSpace=BoneRelative
    alphaBoolBlend {blendOption=Linear}
    boundBone {boneName=dyn_r_shoulder_flower_01}
    componentAppliedLinearAccClamp=<1500 1500 1500>
    componentLinearAccScale=<0.4 0.4 1>
    componentLinearVelScale=<0.2 0.2 0.4>
    gravityOverride {z=-980}
    physicsBodyDefinitions [{
      sphereCollisionRadius=0
      boundBone {boneName=dyn_r_shoulder_flower_01}
      boxExtents=<6 6 6>
      constraintSetup {angularLimitsMax=<5 5 5>, angularLimitsMin=<-5 -5 -5>, angularTarget=<1 0 0>}
      localJointOffset=<0.2 0 0>
     }]
    relativeSpaceBone {boneName=root}
   }
   n10 = Anim Dynamics - Bone: dyn_r_shoulder_flower_02(ComponentPose=@n9, Alpha=@n8, ExternalForce="0,0,0") {
    angularDampingOverride=200
    angularSpringConstant=500
    bAngularSpring=true
    bGravityOverrideInSimSpace=true
    bUseGravityOverride=true
    simulationSpace=BoneRelative
    alphaBoolBlend {blendOption=Linear}
    boundBone {boneName=dyn_r_shoulder_flower_02}
    componentAppliedLinearAccClamp=<1500 1500 1500>
    componentLinearAccScale=<0.4 0.4 1>
    componentLinearVelScale=<0.2 0.2 0.4>
    gravityOverride {z=-980}
    physicsBodyDefinitions [{
      sphereCollisionRadius=0
      boundBone {boneName=dyn_r_shoulder_flower_02}
      boxExtents=<6 6 6>
      constraintSetup {angularLimitsMax=<5 5 5>, angularLimitsMin=<-5 -5 -5>, angularTarget=<1 0 0>}
      localJointOffset=<0.2 0 0>
     }]
    relativeSpaceBone {boneName=root}
   }
   n11 = Anim Dynamics - Bone: dyn_r_shoulder_flower_03(ComponentPose=@n10, Alpha=@n8, ExternalForce="0,0,0") {
    angularDampingOverride=200
    angularSpringConstant=500
    bAngularSpring=true
    bGravityOverrideInSimSpace=true
    bUseGravityOverride=true
    simulationSpace=BoneRelative
    alphaBoolBlend {blendOption=Linear}
    boundBone {boneName=dyn_r_shoulder_flower_03}
    componentAppliedLinearAccClamp=<1500 1500 1500>
    componentLinearAccScale=<0.4 0.4 1>
    componentLinearVelScale=<0.2 0.2 0.4>
    gravityOverride {z=-980}
    physicsBodyDefinitions [{
      sphereCollisionRadius=0
      boundBone {boneName=dyn_r_shoulder_flower_03}
      boxExtents=<6 6 6>
      constraintSetup {angularLimitsMax=<5 5 5>, angularLimitsMin=<-5 -5 -5>, angularTarget=<1 0 0>}
      localJointOffset=<0.2 0 0>
     }]
    relativeSpaceBone {boneName=root}
   }
   n12 = Anim Dynamics - Bone: dyn_r_shoulder_flower_04(ComponentPose=@n11, Alpha=@n8, ExternalForce="0,0,0") {
    angularDampingOverride=200
    angularSpringConstant=500
    bAngularSpring=true
    bGravityOverrideInSimSpace=true
    bUseGravityOverride=true
    simulationSpace=BoneRelative
    alphaBoolBlend {blendOption=Linear}
    boundBone {boneName=dyn_r_shoulder_flower_04}
    componentAppliedLinearAccClamp=<1500 1500 1500>
    componentLinearAccScale=<0.4 0.4 1>
    componentLinearVelScale=<0.2 0.2 0.4>
    gravityOverride {z=-980}
    physicsBodyDefinitions [{
      sphereCollisionRadius=0
      boundBone {boneName=dyn_r_shoulder_flower_04}
      boxExtents=<6 6 6>
      constraintSetup {angularLimitsMax=<5 5 5>, angularLimitsMin=<-5 -5 -5>, angularTarget=<1 0 0>}
      localJointOffset=<0.2 0 0>
     }]
    relativeSpaceBone {boneName=root}
   }
   n13 = Anim Dynamics - Bone: dyn_r_shoulder_flower_05(ComponentPose=@n12, Alpha=@n8, ExternalForce="0,0,0") {
    angularDampingOverride=200
    angularSpringConstant=500
    bAngularSpring=true
    bGravityOverrideInSimSpace=true
    bUseGravityOverride=true
    simulationSpace=BoneRelative
    alphaBoolBlend {blendOption=Linear}
    boundBone {boneName=dyn_r_shoulder_flower_05}
    componentAppliedLinearAccClamp=<1500 1500 1500>
    componentLinearAccScale=<0.4 0.4 1>
    componentLinearVelScale=<0.2 0.2 0.4>
    gravityOverride {z=-980}
    physicsBodyDefinitions [{
      sphereCollisionRadius=0
      boundBone {boneName=dyn_r_shoulder_flower_05}
      boxExtents=<6 6 6>
      constraintSetup {angularLimitsMax=<5 5 5>, angularLimitsMin=<-5 -5 -5>, angularTarget=<1 0 0>}
      localJointOffset=<0.2 0 0>
     }]
    relativeSpaceBone {boneName=root}
   }
   n14 = Anim Dynamics - Bone: dyn_r_shoulder_leaf_01(ComponentPose=@n13, Alpha=@n8, ExternalForce="0,0,0") {
    angularDampingOverride=200
    angularSpringConstant=500
    bAngularSpring=true
    bGravityOverrideInSimSpace=true
    bUseGravityOverride=true
    simulationSpace=BoneRelative
    alphaBoolBlend {blendOption=Linear}
    boundBone {boneName=dyn_r_shoulder_leaf_01}
    componentAppliedLinearAccClamp=<1500 1500 1500>
    componentLinearAccScale=<0.4 0.4 1>
    componentLinearVelScale=<0.2 0.2 0.4>
    gravityOverride {z=-980}
    physicsBodyDefinitions [{
      sphereCollisionRadius=0
      boundBone {boneName=dyn_r_shoulder_leaf_01}
      boxExtents=<7 7 7>
      constraintSetup {angularLimitsMax=<8 5 8>, angularLimitsMin=<-8 -5 -8>, angularTarget=<1 0 0>}
      localJointOffset=<0.2 0 0>
     }]
    relativeSpaceBone {boneName=root}
   }
   n15 = Anim Dynamics - Bone: dyn_r_shoulder_leaf_02(ComponentPose=@n14, Alpha=@n8, ExternalForce="0,0,0") {
    angularDampingOverride=200
    angularSpringConstant=500
    bAngularSpring=true
    bGravityOverrideInSimSpace=true
    bUseGravityOverride=true
    simulationSpace=BoneRelative
    alphaBoolBlend {blendOption=Linear}
    boundBone {boneName=dyn_r_shoulder_leaf_02}
    componentAppliedLinearAccClamp=<1500 1500 1500>
    componentLinearAccScale=<0.4 0.4 1>
    componentLinearVelScale=<0.2 0.2 0.4>
    gravityOverride {z=-980}
    physicsBodyDefinitions [{
      sphereCollisionRadius=0
      boundBone {boneName=dyn_r_shoulder_leaf_02}
      boxExtents=<7 7 7>
      constraintSetup {angularLimitsMax=<8 5 8>, angularLimitsMin=<-8 -5 -8>, angularTarget=<1 0 0>}
      localJointOffset=<0.2 0 0>
     }]
    relativeSpaceBone {boneName=root}
   }
   n16 = Anim Dynamics - Bone: dyn_r_shoulder_leaf_03(ComponentPose=@n15, Alpha=@n8, ExternalForce="0,0,0") {
    angularDampingOverride=200
    angularSpringConstant=500
    bAngularSpring=true
    bGravityOverrideInSimSpace=true
    bUseGravityOverride=true
    simulationSpace=BoneRelative
    alphaBoolBlend {blendOption=Linear}
    boundBone {boneName=dyn_r_shoulder_leaf_03}
    componentAppliedLinearAccClamp=<1500 1500 1500>
    componentLinearAccScale=<0.4 0.4 1>
    componentLinearVelScale=<0.2 0.2 0.4>
    gravityOverride {z=-980}
    physicsBodyDefinitions [{
      sphereCollisionRadius=0
      boundBone {boneName=dyn_r_shoulder_leaf_03}
      boxExtents=<7 7 7>
      constraintSetup {angularLimitsMax=<8 5 8>, angularLimitsMin=<-8 -5 -8>, angularTarget=<1 0 0>}
      localJointOffset=<0.2 0 0>
     }]
    relativeSpaceBone {boneName=root}
   }
   n17 = Anim Dynamics - Bone: dyn_r_shoulder_leaf_04(ComponentPose=@n16, Alpha=@n8, ExternalForce="0,0,0") {
    angularDampingOverride=200
    angularSpringConstant=500
    bAngularSpring=true
    bGravityOverrideInSimSpace=true
    bUseGravityOverride=true
    simulationSpace=BoneRelative
    alphaBoolBlend {blendOption=Linear}
    boundBone {boneName=dyn_r_shoulder_leaf_04}
    componentAppliedLinearAccClamp=<1500 1500 1500>
    componentLinearAccScale=<0.4 0.4 1>
    componentLinearVelScale=<0.2 0.2 0.4>
    gravityOverride {z=-980}
    physicsBodyDefinitions [{
      sphereCollisionRadius=0
      boundBone {boneName=dyn_r_shoulder_leaf_04}
      boxExtents=<7 7 7>
      constraintSetup {angularLimitsMax=<8 5 8>, angularLimitsMin=<-8 -5 -8>, angularTarget=<1 0 0>}
      localJointOffset=<0.2 0 0>
     }]
    relativeSpaceBone {boneName=root}
   }
   n18 = Anim Dynamics - Bone: dyn_r_shoulder_leaf_05(ComponentPose=@n17, Alpha=@n8, ExternalForce="0,0,0") {
    angularDampingOverride=200
    angularSpringConstant=500
    bAngularSpring=true
    bGravityOverrideInSimSpace=true
    bUseGravityOverride=true
    simulationSpace=BoneRelative
    alphaBoolBlend {blendOption=Linear}
    boundBone {boneName=dyn_r_shoulder_leaf_05}
    componentAppliedLinearAccClamp=<1500 1500 1500>
    componentLinearAccScale=<0.4 0.4 1>
    componentLinearVelScale=<0.2 0.2 0.4>
    gravityOverride {z=-980}
    physicsBodyDefinitions [{
      sphereCollisionRadius=0
      boundBone {boneName=dyn_r_shoulder_leaf_05}
      boxExtents=<7 7 7>
      constraintSetup {angularLimitsMax=<8 5 8>, angularLimitsMin=<-8 -5 -8>, angularTarget=<1 0 0>}
      localJointOffset=<0.2 0 0>
     }]
    relativeSpaceBone {boneName=root}
   }
   n19 = Anim Dynamics - Bone: dyn_r_shoulder_leaf_06(ComponentPose=@n18, Alpha=@n8, ExternalForce="0,0,0") {
    angularDampingOverride=200
    angularSpringConstant=500
    bAngularSpring=true
    bGravityOverrideInSimSpace=true
    bUseGravityOverride=true
    simulationSpace=BoneRelative
    alphaBoolBlend {blendOption=Linear}
    boundBone {boneName=dyn_r_shoulder_leaf_06}
    componentAppliedLinearAccClamp=<1500 1500 1500>
    componentLinearAccScale=<0.4 0.4 1>
    componentLinearVelScale=<0.2 0.2 0.4>
    gravityOverride {z=-980}
    physicsBodyDefinitions [{
      sphereCollisionRadius=0
      boundBone {boneName=dyn_r_shoulder_leaf_06}
      boxExtents=<7 7 7>
      constraintSetup {angularLimitsMax=<8 5 8>, angularLimitsMin=<-8 -5 -8>, angularTarget=<1 0 0>}
      localJointOffset=<0.2 0 0>
     }]
    relativeSpaceBone {boneName=root}
   }
   n20 = Anim Dynamics - Bone: dyn_r_shoulder_leaf_07(ComponentPose=@n19, Alpha=@n8, ExternalForce="0,0,0") {
    angularDampingOverride=200
    angularSpringConstant=500
    bAngularSpring=true
    bGravityOverrideInSimSpace=true
    bUseGravityOverride=true
    simulationSpace=BoneRelative
    alphaBoolBlend {blendOption=Linear}
    boundBone {boneName=dyn_r_shoulder_leaf_07}
    componentAppliedLinearAccClamp=<1500 1500 1500>
    componentLinearAccScale=<0.4 0.4 1>
    componentLinearVelScale=<0.2 0.2 0.4>
    gravityOverride {z=-980}
    physicsBodyDefinitions [{
      sphereCollisionRadius=0
      boundBone {boneName=dyn_r_shoulder_leaf_07}
      boxExtents=<7 7 7>
      constraintSetup {angularLimitsMax=<8 5 8>, angularLimitsMin=<-8 -5 -8>, angularTarget=<1 0 0>}
      localJointOffset=<0.2 0 0>
     }]
    relativeSpaceBone {boneName=root}
   }
   n21 = Anim Dynamics - Bone: dyn_r_shoulder_leaf_08(ComponentPose=@n20, Alpha=@n8, ExternalForce="0,0,0") {
    angularDampingOverride=200
    angularSpringConstant=500
    bAngularSpring=true
    bGravityOverrideInSimSpace=true
    bUseGravityOverride=true
    simulationSpace=BoneRelative
    alphaBoolBlend {blendOption=Linear}
    boundBone {boneName=dyn_r_shoulder_leaf_08}
    componentAppliedLinearAccClamp=<1500 1500 1500>
    componentLinearAccScale=<0.4 0.4 1>
    componentLinearVelScale=<0.2 0.2 0.4>
    gravityOverride {z=-980}
    physicsBodyDefinitions [{
      sphereCollisionRadius=0
      boundBone {boneName=dyn_r_shoulder_leaf_08}
      boxExtents=<7 7 7>
      constraintSetup {angularLimitsMax=<8 5 8>, angularLimitsMin=<-8 -5 -8>, angularTarget=<1 0 0>}
      localJointOffset=<0.2 0 0>
     }]
    relativeSpaceBone {boneName=root}
   }
   n22 = Anim Dynamics - Bone: dyn_r_shoulder_leaf_09(ComponentPose=@n21, Alpha=@n8, ExternalForce="0,0,0") {
    angularDampingOverride=200
    angularSpringConstant=500
    bAngularSpring=true
    bGravityOverrideInSimSpace=true
    bUseGravityOverride=true
    simulationSpace=BoneRelative
    alphaBoolBlend {blendOption=Linear}
    boundBone {boneName=dyn_r_shoulder_leaf_09}
    componentAppliedLinearAccClamp=<1500 1500 1500>
    componentLinearAccScale=<0.4 0.4 1>
    componentLinearVelScale=<0.2 0.2 0.4>
    gravityOverride {z=-980}
    physicsBodyDefinitions [{
      sphereCollisionRadius=0
      boundBone {boneName=dyn_r_shoulder_leaf_09}
      boxExtents=<7 7 7>
      constraintSetup {angularLimitsMax=<8 5 8>, angularLimitsMin=<-8 -5 -8>, angularTarget=<1 0 0>}
      localJointOffset=<0.2 0 0>
     }]
    relativeSpaceBone {boneName=root}
   }
   n23 = Anim Dynamics - Bone: dyn_r_shoulder_leaf_10(ComponentPose=@n22, Alpha=@n8, ExternalForce="0,0,0") {
    angularDampingOverride=200
    angularSpringConstant=500
    bAngularSpring=true
    bGravityOverrideInSimSpace=true
    bUseGravityOverride=true
    simulationSpace=BoneRelative
    alphaBoolBlend {blendOption=Linear}
    boundBone {boneName=dyn_r_shoulder_leaf_10}
    componentAppliedLinearAccClamp=<1500 1500 1500>
    componentLinearAccScale=<0.4 0.4 1>
    componentLinearVelScale=<0.2 0.2 0.4>
    gravityOverride {z=-980}
    physicsBodyDefinitions [{
      sphereCollisionRadius=0
      boundBone {boneName=dyn_r_shoulder_leaf_10}
      boxExtents=<7 7 7>
      constraintSetup {angularLimitsMax=<8 5 8>, angularLimitsMin=<-8 -5 -8>, angularTarget=<1 0 0>}
      localJointOffset=<0.2 0 0>
     }]
    relativeSpaceBone {boneName=root}
   }
   n24 = Anim Dynamics - Bone: dyn_r_shoulder_leaf_11(ComponentPose=@n23, Alpha=@n8, ExternalForce="0,0,0") {
    angularDampingOverride=200
    angularSpringConstant=500
    bAngularSpring=true
    bGravityOverrideInSimSpace=true
    bUseGravityOverride=true
    simulationSpace=BoneRelative
    alphaBoolBlend {blendOption=Linear}
    boundBone {boneName=dyn_r_shoulder_leaf_11}
    componentAppliedLinearAccClamp=<1500 1500 1500>
    componentLinearAccScale=<0.4 0.4 1>
    componentLinearVelScale=<0.2 0.2 0.4>
    gravityOverride {z=-980}
    physicsBodyDefinitions [{
      sphereCollisionRadius=0
      boundBone {boneName=dyn_r_shoulder_leaf_11}
      boxExtents=<7 7 7>
      constraintSetup {angularLimitsMax=<8 5 8>, angularLimitsMin=<-8 -5 -8>, angularTarget=<1 0 0>}
      localJointOffset=<0.2 0 0>
     }]
    relativeSpaceBone {boneName=root}
   }
   n25 = Anim Dynamics - Bone: dyn_r_shoulder_leaf_12(ComponentPose=@n24, Alpha=@n8, ExternalForce="0,0,0") {
    angularDampingOverride=200
    angularSpringConstant=500
    bAngularSpring=true
    bGravityOverrideInSimSpace=true
    bUseGravityOverride=true
    simulationSpace=BoneRelative
    alphaBoolBlend {blendOption=Linear}
    boundBone {boneName=dyn_r_shoulder_leaf_12}
    componentAppliedLinearAccClamp=<1500 1500 1500>
    componentLinearAccScale=<0.4 0.4 1>
    componentLinearVelScale=<0.2 0.2 0.4>
    gravityOverride {z=-980}
    physicsBodyDefinitions [{
      sphereCollisionRadius=0
      boundBone {boneName=dyn_r_shoulder_leaf_12}
      boxExtents=<7 7 7>
      constraintSetup {angularLimitsMax=<8 5 8>, angularLimitsMin=<-8 -5 -8>, angularTarget=<1 0 0>}
      localJointOffset=<0.2 0 0>
     }]
    relativeSpaceBone {boneName=root}
   }
   n26 = Anim Dynamics - Bone: dyn_r_shoulder_leaf_13(ComponentPose=@n25, Alpha=@n8, ExternalForce="0,0,0") {
    angularDampingOverride=200
    angularSpringConstant=500
    bAngularSpring=true
    bGravityOverrideInSimSpace=true
    bUseGravityOverride=true
    simulationSpace=BoneRelative
    alphaBoolBlend {blendOption=Linear}
    boundBone {boneName=dyn_r_shoulder_leaf_13}
    componentAppliedLinearAccClamp=<1500 1500 1500>
    componentLinearAccScale=<0.4 0.4 1>
    componentLinearVelScale=<0.2 0.2 0.4>
    gravityOverride {z=-980}
    physicsBodyDefinitions [{
      sphereCollisionRadius=0
      boundBone {boneName=dyn_r_shoulder_leaf_13}
      boxExtents=<7 7 7>
      constraintSetup {angularLimitsMax=<8 5 8>, angularLimitsMin=<-8 -5 -8>, angularTarget=<1 0 0>}
      localJointOffset=<0.2 0 0>
     }]
    relativeSpaceBone {boneName=root}
   }
   n27 = Anim Dynamics - Bone: dyn_r_shoulder_leaf_14(ComponentPose=@n26, Alpha=@n8, ExternalForce="0,0,0") {
    angularDampingOverride=200
    angularSpringConstant=500
    bAngularSpring=true
    bGravityOverrideInSimSpace=true
    bUseGravityOverride=true
    simulationSpace=BoneRelative
    alphaBoolBlend {blendOption=Linear}
    boundBone {boneName=dyn_r_shoulder_leaf_14}
    componentAppliedLinearAccClamp=<1500 1500 1500>
    componentLinearAccScale=<0.4 0.4 1>
    componentLinearVelScale=<0.2 0.2 0.4>
    gravityOverride {z=-980}
    physicsBodyDefinitions [{
      sphereCollisionRadius=0
      boundBone {boneName=dyn_r_shoulder_leaf_14}
      boxExtents=<7 7 7>
      constraintSetup {angularLimitsMax=<8 -10 -20>, angularLimitsMin=<-8 -20 -30>, angularTarget=<1 0 0>}
      localJointOffset=<0.2 0 0>
     }]
    relativeSpaceBone {boneName=root}
   }
   n28 = Anim Dynamics - Bone: dyn_r_shoulder_leaf_15(ComponentPose=@n27, Alpha=@n8, ExternalForce="0,0,0") {
    angularDampingOverride=200
    angularSpringConstant=500
    bAngularSpring=true
    bGravityOverrideInSimSpace=true
    bUseGravityOverride=true
    simulationSpace=BoneRelative
    alphaBoolBlend {blendOption=Linear}
    boundBone {boneName=dyn_r_shoulder_leaf_15}
    componentAppliedLinearAccClamp=<1500 1500 1500>
    componentLinearAccScale=<0.4 0.4 1>
    componentLinearVelScale=<0.2 0.2 0.4>
    gravityOverride {z=-980}
    physicsBodyDefinitions [{
      sphereCollisionRadius=0
      boundBone {boneName=dyn_r_shoulder_leaf_15}
      boxExtents=<7 7 7>
      constraintSetup {angularLimitsMax=<8 5 8>, angularLimitsMin=<-8 -5 -8>, angularTarget=<1 0 0>}
      localJointOffset=<0.2 0 0>
     }]
    relativeSpaceBone {boneName=root}
   }
   n29 = Anim Dynamics - Bone: dyn_r_shoulder_leaf_16(ComponentPose=@n28, Alpha=@n8, ExternalForce="0,0,0") {
    angularDampingOverride=200
    angularSpringConstant=500
    bAngularSpring=true
    bGravityOverrideInSimSpace=true
    bUseGravityOverride=true
    simulationSpace=BoneRelative
    alphaBoolBlend {blendOption=Linear}
    boundBone {boneName=dyn_r_shoulder_leaf_16}
    componentAppliedLinearAccClamp=<1500 1500 1500>
    componentLinearAccScale=<0.4 0.4 1>
    componentLinearVelScale=<0.2 0.2 0.4>
    gravityOverride {z=-980}
    physicsBodyDefinitions [{
      sphereCollisionRadius=0
      boundBone {boneName=dyn_r_shoulder_leaf_16}
      boxExtents=<7 7 7>
      constraintSetup {angularLimitsMax=<8 5 8>, angularLimitsMin=<-8 -5 -8>, angularTarget=<1 0 0>}
      localJointOffset=<0.2 0 0>
     }]
    relativeSpaceBone {boneName=root}
   }
   n30 = Anim Dynamics - Bone: dyn_r_shoulder_leaf_17(ComponentPose=@n29, Alpha=@n8, ExternalForce="0,0,0") {
    angularDampingOverride=200
    angularSpringConstant=500
    bAngularSpring=true
    bGravityOverrideInSimSpace=true
    bUseGravityOverride=true
    simulationSpace=BoneRelative
    alphaBoolBlend {blendOption=Linear}
    boundBone {boneName=dyn_r_shoulder_leaf_17}
    componentAppliedLinearAccClamp=<1500 1500 1500>
    componentLinearAccScale=<0.4 0.4 1>
    componentLinearVelScale=<0.2 0.2 0.4>
    gravityOverride {z=-980}
    physicsBodyDefinitions [{
      sphereCollisionRadius=0
      boundBone {boneName=dyn_r_shoulder_leaf_17}
      boxExtents=<7 7 7>
      constraintSetup {angularLimitsMax=<8 5 8>, angularLimitsMin=<-8 -5 -8>, angularTarget=<1 0 0>}
      localJointOffset=<0.2 0 0>
     }]
    relativeSpaceBone {boneName=root}
   }
   n31 = Anim Dynamics - Bone: dyn_r_shoulder_leaf_18(ComponentPose=@n30, Alpha=@n8, ExternalForce="0,0,0") {
    angularDampingOverride=200
    angularSpringConstant=500
    bAngularSpring=true
    bGravityOverrideInSimSpace=true
    bUseGravityOverride=true
    simulationSpace=BoneRelative
    alphaBoolBlend {blendOption=Linear}
    boundBone {boneName=dyn_r_shoulder_leaf_18}
    componentAppliedLinearAccClamp=<1500 1500 1500>
    componentLinearAccScale=<0.4 0.4 1>
    componentLinearVelScale=<0.2 0.2 0.4>
    gravityOverride {z=-980}
    physicsBodyDefinitions [{
      sphereCollisionRadius=0
      boundBone {boneName=dyn_r_shoulder_leaf_18}
      boxExtents=<7 7 7>
      constraintSetup {angularLimitsMax=<8 5 -20>, angularLimitsMin=<-8 -5 -30>, angularTarget=<1 0 0>}
      localJointOffset=<0.2 0 0>
     }]
    relativeSpaceBone {boneName=root}
   }
   n32 = Anim Dynamics - Bone: dyn_r_shoulder_leaf_19(ComponentPose=@n31, Alpha=@n8, ExternalForce="0,0,0") {
    angularDampingOverride=200
    angularSpringConstant=500
    bAngularSpring=true
    bGravityOverrideInSimSpace=true
    bUseGravityOverride=true
    simulationSpace=BoneRelative
    alphaBoolBlend {blendOption=Linear}
    boundBone {boneName=dyn_r_shoulder_leaf_19}
    componentAppliedLinearAccClamp=<1500 1500 1500>
    componentLinearAccScale=<0.4 0.4 1>
    componentLinearVelScale=<0.2 0.2 0.4>
    gravityOverride {z=-980}
    physicsBodyDefinitions [{
      sphereCollisionRadius=0
      boundBone {boneName=dyn_r_shoulder_leaf_19}
      boxExtents=<7 7 7>
      constraintSetup {angularLimitsMax=<8 5 8>, angularLimitsMin=<-8 -5 -8>, angularTarget=<1 0 0>}
      localJointOffset=<0.2 0 0>
     }]
    relativeSpaceBone {boneName=root}
   }
   n33 = Anim Dynamics - Bone: dyn_r_arm_leaf_01(ComponentPose=@n32, Alpha=@n8, ExternalForce="0,0,0") {
    angularDampingOverride=200
    angularSpringConstant=500
    bAngularSpring=true
    bGravityOverrideInSimSpace=true
    bUseGravityOverride=true
    simulationSpace=BoneRelative
    alphaBoolBlend {blendOption=Linear}
    boundBone {boneName=dyn_r_arm_leaf_01}
    componentAppliedLinearAccClamp=<1500 1500 1500>
    componentLinearAccScale=<0.4 0.4 1>
    componentLinearVelScale=<0.2 0.2 0.4>
    gravityOverride {z=-980}
    physicsBodyDefinitions [{
      sphereCollisionRadius=0
      boundBone {boneName=dyn_r_arm_leaf_01}
      boxExtents=<7 7 7>
      constraintSetup {angularLimitsMax=<8 8 8>, angularLimitsMin=<-8 -8 -8>, angularTarget=<1 0 0>}
      localJointOffset=<0.2 0 0>
     }]
    relativeSpaceBone {boneName=root}
   }
   n34 = Anim Dynamics - Bone: dyn_r_arm_leaf_02(ComponentPose=@n33, Alpha=@n8, ExternalForce="0,0,0") {
    angularDampingOverride=200
    angularSpringConstant=500
    bAngularSpring=true
    bGravityOverrideInSimSpace=true
    bUseGravityOverride=true
    simulationSpace=BoneRelative
    alphaBoolBlend {blendOption=Linear}
    boundBone {boneName=dyn_r_arm_leaf_02}
    componentAppliedLinearAccClamp=<1500 1500 1500>
    componentLinearAccScale=<0.4 0.4 1>
    componentLinearVelScale=<0.2 0.2 0.4>
    gravityOverride {z=-980}
    physicsBodyDefinitions [{
      sphereCollisionRadius=0
      boundBone {boneName=dyn_r_arm_leaf_02}
      boxExtents=<7 7 7>
      constraintSetup {angularLimitsMax=<8 8 8>, angularLimitsMin=<-8 -8 -8>, angularTarget=<1 0 0>}
      localJointOffset=<0.2 0 0>
     }]
    relativeSpaceBone {boneName=root}
   }
   n35 = Anim Dynamics - Bone: dyn_r_arm_leaf_03(ComponentPose=@n34, Alpha=@n8, ExternalForce="0,0,0") {
    angularDampingOverride=200
    angularSpringConstant=500
    bAngularSpring=true
    bGravityOverrideInSimSpace=true
    bUseGravityOverride=true
    simulationSpace=BoneRelative
    alphaBoolBlend {blendOption=Linear}
    boundBone {boneName=dyn_r_arm_leaf_03}
    componentAppliedLinearAccClamp=<1500 1500 1500>
    componentLinearAccScale=<0.4 0.4 1>
    componentLinearVelScale=<0.2 0.2 0.4>
    gravityOverride {z=-980}
    physicsBodyDefinitions [{
      sphereCollisionRadius=0
      boundBone {boneName=dyn_r_arm_leaf_03}
      boxExtents=<7 7 7>
      constraintSetup {angularLimitsMax=<8 8 8>, angularLimitsMin=<-8 -8 -8>, angularTarget=<1 0 0>}
      localJointOffset=<0.2 0 0>
     }]
    relativeSpaceBone {boneName=root}
   }
   n36 = Anim Dynamics - Bone: dyn_r_arm_leaf_04(ComponentPose=@n35, Alpha=@n8, ExternalForce="0,0,0") {
    angularDampingOverride=200
    angularSpringConstant=500
    bAngularSpring=true
    bGravityOverrideInSimSpace=true
    bUseGravityOverride=true
    simulationSpace=BoneRelative
    alphaBoolBlend {blendOption=Linear}
    boundBone {boneName=dyn_r_arm_leaf_04}
    componentAppliedLinearAccClamp=<1500 1500 1500>
    componentLinearAccScale=<0.4 0.4 1>
    componentLinearVelScale=<0.2 0.2 0.4>
    gravityOverride {z=-980}
    physicsBodyDefinitions [{
      sphereCollisionRadius=0
      boundBone {boneName=dyn_r_arm_leaf_04}
      boxExtents=<7 7 7>
      constraintSetup {angularLimitsMax=<8 8 8>, angularLimitsMin=<-8 -8 -8>, angularTarget=<1 0 0>}
      localJointOffset=<0.2 0 0>
     }]
    relativeSpaceBone {boneName=root}
   }
   n37 = Anim Dynamics - Bone: dyn_r_hand_flower_01(ComponentPose=@n36, Alpha=@n8, ExternalForce="0,0,0") {
    angularDampingOverride=200
    angularSpringConstant=500
    bAngularSpring=true
    bGravityOverrideInSimSpace=true
    bUseGravityOverride=true
    simulationSpace=BoneRelative
    alphaBoolBlend {blendOption=Linear}
    boundBone {boneName=dyn_r_hand_flower_01}
    componentAppliedLinearAccClamp=<1500 1500 1500>
    componentLinearAccScale=<0.4 0.4 1>
    componentLinearVelScale=<0.2 0.2 0.4>
    gravityOverride {z=-980}
    physicsBodyDefinitions [{
      sphereCollisionRadius=0
      boundBone {boneName=dyn_r_hand_flower_01}
      boxExtents=<6 6 6>
      constraintSetup {angularLimitsMax=<5 5 5>, angularLimitsMin=<-5 -5 -5>, angularTarget=<1 0 0>}
      localJointOffset=<0.2 0 0>
     }]
    relativeSpaceBone {boneName=root}
   }
   n38 = Anim Dynamics - Bone: dyn_r_hand_flower_02(ComponentPose=@n37, Alpha=@n8, ExternalForce="0,0,0") {
    angularDampingOverride=200
    angularSpringConstant=500
    bAngularSpring=true
    bGravityOverrideInSimSpace=true
    bUseGravityOverride=true
    simulationSpace=BoneRelative
    alphaBoolBlend {blendOption=Linear}
    boundBone {boneName=dyn_r_hand_flower_02}
    componentAppliedLinearAccClamp=<1500 1500 1500>
    componentLinearAccScale=<0.4 0.4 1>
    componentLinearVelScale=<0.2 0.2 0.4>
    gravityOverride {z=-980}
    physicsBodyDefinitions [{
      sphereCollisionRadius=0
      boundBone {boneName=dyn_r_hand_flower_02}
      boxExtents=<6 6 6>
      constraintSetup {angularLimitsMax=<5 5 5>, angularLimitsMin=<-5 -5 -5>, angularTarget=<1 0 0>}
      localJointOffset=<0.2 0 0>
     }]
    relativeSpaceBone {boneName=root}
   }
   n39 = Get ADS_Aim()
   n40 = Anim Dynamics - Bone: dyn_l_shoulder_flower_01(ComponentPose=@n38, Alpha=@n39, ExternalForce="0,0,0") {
    angularDampingOverride=200
    angularSpringConstant=500
    bAngularSpring=true
    bGravityOverrideInSimSpace=true
    bUseGravityOverride=true
    simulationSpace=BoneRelative
    alphaBoolBlend {blendOption=Linear}
    boundBone {boneName=dyn_l_shoulder_flower_01}
    componentAppliedLinearAccClamp=<1500 1500 1500>
    componentLinearAccScale=<0.4 0.4 1>
    componentLinearVelScale=<0.2 0.2 0.4>
    gravityOverride {z=-980}
    physicsBodyDefinitions [{
      sphereCollisionRadius=0
      boundBone {boneName=dyn_l_shoulder_flower_01}
      boxExtents=<6 6 6>
      constraintSetup {angularLimitsMax=<5 5 5>, angularLimitsMin=<-5 -5 -5>, angularTarget=<1 0 0>}
      localJointOffset=<0.2 0 0>
     }]
    relativeSpaceBone {boneName=root}
   }
   n41 = Anim Dynamics - Bone: dyn_l_shoulder_flower_02(ComponentPose=@n40, Alpha=@n39, ExternalForce="0,0,0") {
    angularDampingOverride=200
    angularSpringConstant=500
    bAngularSpring=true
    bGravityOverrideInSimSpace=true
    bUseGravityOverride=true
    simulationSpace=BoneRelative
    alphaBoolBlend {blendOption=Linear}
    boundBone {boneName=dyn_l_shoulder_flower_02}
    componentAppliedLinearAccClamp=<1500 1500 1500>
    componentLinearAccScale=<0.4 0.4 1>
    componentLinearVelScale=<0.2 0.2 0.4>
    gravityOverride {z=-980}
    physicsBodyDefinitions [{
      sphereCollisionRadius=0
      boundBone {boneName=dyn_l_shoulder_flower_02}
      boxExtents=<6 6 6>
      constraintSetup {angularLimitsMax=<5 5 5>, angularLimitsMin=<-5 -5 -5>, angularTarget=<1 0 0>}
      localJointOffset=<0.2 0 0>
     }]
    relativeSpaceBone {boneName=root}
   }
   n42 = Anim Dynamics - Bone: dyn_l_shoulder_flower_03(ComponentPose=@n41, Alpha=@n39, ExternalForce="0,0,0") {
    angularDampingOverride=200
    angularSpringConstant=500
    bAngularSpring=true
    bGravityOverrideInSimSpace=true
    bUseGravityOverride=true
    simulationSpace=BoneRelative
    alphaBoolBlend {blendOption=Linear}
    boundBone {boneName=dyn_l_shoulder_flower_03}
    componentAppliedLinearAccClamp=<1500 1500 1500>
    componentLinearAccScale=<0.4 0.4 1>
    componentLinearVelScale=<0.2 0.2 0.4>
    gravityOverride {z=-980}
    physicsBodyDefinitions [{
      sphereCollisionRadius=0
      boundBone {boneName=dyn_l_shoulder_flower_03}
      boxExtents=<6 6 6>
      constraintSetup {angularLimitsMax=<5 5 5>, angularLimitsMin=<-5 -5 -5>, angularTarget=<1 0 0>}
      localJointOffset=<0.2 0 0>
     }]
    relativeSpaceBone {boneName=root}
   }
   n43 = Anim Dynamics - Bone: dyn_l_shoulder_flower_04(ComponentPose=@n42, Alpha=@n39, ExternalForce="0,0,0") {
    angularDampingOverride=200
    angularSpringConstant=500
    bAngularSpring=true
    bGravityOverrideInSimSpace=true
    bUseGravityOverride=true
    simulationSpace=BoneRelative
    alphaBoolBlend {blendOption=Linear}
    boundBone {boneName=dyn_l_shoulder_flower_04}
    componentAppliedLinearAccClamp=<1500 1500 1500>
    componentLinearAccScale=<0.4 0.4 1>
    componentLinearVelScale=<0.2 0.2 0.4>
    gravityOverride {z=-980}
    physicsBodyDefinitions [{
      sphereCollisionRadius=0
      boundBone {boneName=dyn_l_shoulder_flower_04}
      boxExtents=<6 6 6>
      constraintSetup {angularLimitsMax=<5 5 5>, angularLimitsMin=<-5 -5 -5>, angularTarget=<1 0 0>}
      localJointOffset=<0.2 0 0>
     }]
    relativeSpaceBone {boneName=root}
   }
   n44 = Anim Dynamics - Bone: dyn_l_shoulder_flower_05(ComponentPose=@n43, Alpha=@n39, ExternalForce="0,0,0") {
    angularDampingOverride=200
    angularSpringConstant=500
    bAngularSpring=true
    bGravityOverrideInSimSpace=true
    bUseGravityOverride=true
    simulationSpace=BoneRelative
    alphaBoolBlend {blendOption=Linear}
    boundBone {boneName=dyn_l_shoulder_flower_05}
    componentAppliedLinearAccClamp=<1500 1500 1500>
    componentLinearAccScale=<0.4 0.4 1>
    componentLinearVelScale=<0.2 0.2 0.4>
    gravityOverride {z=-980}
    physicsBodyDefinitions [{
      sphereCollisionRadius=0
      boundBone {boneName=dyn_l_shoulder_flower_05}
      boxExtents=<6 6 6>
      constraintSetup {angularLimitsMax=<5 5 5>, angularLimitsMin=<-5 -5 -5>, angularTarget=<1 0 0>}
      localJointOffset=<0.2 0 0>
     }]
    relativeSpaceBone {boneName=root}
   }
   n45 = Anim Dynamics - Bone: dyn_l_shoulder_flower_06(ComponentPose=@n44, Alpha=@n39, ExternalForce="0,0,0") {
    angularDampingOverride=200
    angularSpringConstant=500
    bAngularSpring=true
    bGravityOverrideInSimSpace=true
    bUseGravityOverride=true
    simulationSpace=BoneRelative
    alphaBoolBlend {blendOption=Linear}
    boundBone {boneName=dyn_l_shoulder_flower_06}
    componentAppliedLinearAccClamp=<1500 1500 1500>
    componentLinearAccScale=<0.4 0.4 1>
    componentLinearVelScale=<0.2 0.2 0.4>
    gravityOverride {z=-980}
    physicsBodyDefinitions [{
      sphereCollisionRadius=0
      boundBone {boneName=dyn_l_shoulder_flower_06}
      boxExtents=<6 6 6>
      constraintSetup {angularLimitsMax=<5 5 5>, angularLimitsMin=<-5 -5 -5>, angularTarget=<1 0 0>}
      localJointOffset=<0.2 0 0>
     }]
    relativeSpaceBone {boneName=root}
   }
   n46 = Anim Dynamics - Bone: dyn_l_shoulder_flower_07(ComponentPose=@n45, Alpha=@n39, ExternalForce="0,0,0") {
    angularDampingOverride=200
    angularSpringConstant=500
    bAngularSpring=true
    bGravityOverrideInSimSpace=true
    bUseGravityOverride=true
    simulationSpace=BoneRelative
    alphaBoolBlend {blendOption=Linear}
    boundBone {boneName=dyn_l_shoulder_flower_07}
    componentAppliedLinearAccClamp=<1500 1500 1500>
    componentLinearAccScale=<0.4 0.4 1>
    componentLinearVelScale=<0.2 0.2 0.4>
    gravityOverride {z=-980}
    physicsBodyDefinitions [{
      sphereCollisionRadius=0
      boundBone {boneName=dyn_l_shoulder_flower_07}
      boxExtents=<6 6 6>
      constraintSetup {angularLimitsMax=<5 5 5>, angularLimitsMin=<-5 -5 -5>, angularTarget=<1 0 0>}
      localJointOffset=<0.2 0 0>
     }]
    relativeSpaceBone {boneName=root}
   }
   n47 = Anim Dynamics - Bone: dyn_l_shoulder_flower_08(ComponentPose=@n46, Alpha=@n39, ExternalForce="0,0,0") {
    angularDampingOverride=200
    angularSpringConstant=500
    bAngularSpring=true
    bGravityOverrideInSimSpace=true
    bUseGravityOverride=true
    simulationSpace=BoneRelative
    alphaBoolBlend {blendOption=Linear}
    boundBone {boneName=dyn_l_shoulder_flower_08}
    componentAppliedLinearAccClamp=<1500 1500 1500>
    componentLinearAccScale=<0.4 0.4 1>
    componentLinearVelScale=<0.2 0.2 0.4>
    gravityOverride {z=-980}
    physicsBodyDefinitions [{
      sphereCollisionRadius=0
      boundBone {boneName=dyn_l_shoulder_flower_08}
      boxExtents=<6 6 6>
      constraintSetup {angularLimitsMax=<5 5 5>, angularLimitsMin=<-5 -5 -5>, angularTarget=<1 0 0>}
      localJointOffset=<0.2 0 0>
     }]
    relativeSpaceBone {boneName=root}
   }
   n48 = Anim Dynamics - Bone: dyn_l_shoulder_flower_09(ComponentPose=@n47, Alpha=@n39, ExternalForce="0,0,0") {
    angularDampingOverride=200
    angularSpringConstant=500
    bAngularSpring=true
    bGravityOverrideInSimSpace=true
    bUseGravityOverride=true
    simulationSpace=BoneRelative
    alphaBoolBlend {blendOption=Linear}
    boundBone {boneName=dyn_l_shoulder_flower_09}
    componentAppliedLinearAccClamp=<1500 1500 1500>
    componentLinearAccScale=<0.4 0.4 1>
    componentLinearVelScale=<0.2 0.2 0.4>
    gravityOverride {z=-980}
    physicsBodyDefinitions [{
      sphereCollisionRadius=0
      boundBone {boneName=dyn_l_shoulder_flower_09}
      boxExtents=<6 6 6>
      constraintSetup {angularLimitsMax=<5 5 5>, angularLimitsMin=<-5 -5 -5>, angularTarget=<1 0 0>}
      localJointOffset=<0.2 0 0>
     }]
    relativeSpaceBone {boneName=root}
   }
   n49 = Anim Dynamics - Bone: dyn_l_shoulder_leaf_01(ComponentPose=@n48, Alpha=@n39, ExternalForce="0,0,0") {
    angularDampingOverride=200
    angularSpringConstant=500
    bAngularSpring=true
    bGravityOverrideInSimSpace=true
    bUseGravityOverride=true
    simulationSpace=BoneRelative
    alphaBoolBlend {blendOption=Linear}
    boundBone {boneName=dyn_l_shoulder_leaf_01}
    componentAppliedLinearAccClamp=<1500 1500 1500>
    componentLinearAccScale=<0.4 0.4 1>
    componentLinearVelScale=<0.2 0.2 0.4>
    gravityOverride {z=-980}
    physicsBodyDefinitions [{
      sphereCollisionRadius=0
      boundBone {boneName=dyn_l_shoulder_leaf_01}
      boxExtents=<7 7 7>
      constraintSetup {angularLimitsMax=<8 5 8>, angularLimitsMin=<-8 -5 -8>, angularTarget=<1 0 0>}
      localJointOffset=<0.2 0 0>
     }]
    relativeSpaceBone {boneName=root}
   }
   n50 = Anim Dynamics - Bone: dyn_l_shoulder_leaf_02(ComponentPose=@n49, Alpha=@n39, ExternalForce="0,0,0") {
    angularDampingOverride=200
    angularSpringConstant=500
    bAngularSpring=true
    bGravityOverrideInSimSpace=true
    bUseGravityOverride=true
    simulationSpace=BoneRelative
    alphaBoolBlend {blendOption=Linear}
    boundBone {boneName=dyn_l_shoulder_leaf_02}
    componentAppliedLinearAccClamp=<1500 1500 1500>
    componentLinearAccScale=<0.4 0.4 1>
    componentLinearVelScale=<0.2 0.2 0.4>
    gravityOverride {z=-980}
    physicsBodyDefinitions [{
      sphereCollisionRadius=0
      boundBone {boneName=dyn_l_shoulder_leaf_02}
      boxExtents=<7 7 7>
      constraintSetup {angularLimitsMax=<8 5 8>, angularLimitsMin=<-8 -5 -8>, angularTarget=<1 0 0>}
      localJointOffset=<0.2 0 0>
     }]
    relativeSpaceBone {boneName=root}
   }
   n51 = Anim Dynamics - Bone: dyn_l_shoulder_leaf_03(ComponentPose=@n50, Alpha=@n39, ExternalForce="0,0,0") {
    angularDampingOverride=200
    angularSpringConstant=500
    bAngularSpring=true
    bGravityOverrideInSimSpace=true
    bUseGravityOverride=true
    simulationSpace=BoneRelative
    alphaBoolBlend {blendOption=Linear}
    boundBone {boneName=dyn_l_shoulder_leaf_03}
    componentAppliedLinearAccClamp=<1500 1500 1500>
    componentLinearAccScale=<0.4 0.4 1>
    componentLinearVelScale=<0.2 0.2 0.4>
    gravityOverride {z=-980}
    physicsBodyDefinitions [{
      sphereCollisionRadius=0
      boundBone {boneName=dyn_l_shoulder_leaf_03}
      boxExtents=<7 7 7>
      constraintSetup {angularLimitsMax=<8 5 8>, angularLimitsMin=<-8 -5 -8>, angularTarget=<1 0 0>}
      localJointOffset=<0.2 0 0>
     }]
    relativeSpaceBone {boneName=root}
   }
   n52 = Anim Dynamics - Bone: dyn_l_shoulder_leaf_04(ComponentPose=@n51, Alpha=@n39, ExternalForce="0,0,0") {
    angularDampingOverride=200
    angularSpringConstant=500
    bAngularSpring=true
    bGravityOverrideInSimSpace=true
    bUseGravityOverride=true
    simulationSpace=BoneRelative
    alphaBoolBlend {blendOption=Linear}
    boundBone {boneName=dyn_l_shoulder_leaf_04}
    componentAppliedLinearAccClamp=<1500 1500 1500>
    componentLinearAccScale=<0.4 0.4 1>
    componentLinearVelScale=<0.2 0.2 0.4>
    gravityOverride {z=-980}
    physicsBodyDefinitions [{
      sphereCollisionRadius=0
      boundBone {boneName=dyn_l_shoulder_leaf_04}
      boxExtents=<7 7 7>
      constraintSetup {angularLimitsMax=<8 5 8>, angularLimitsMin=<-8 -5 -8>, angularTarget=<1 0 0>}
      localJointOffset=<0.2 0 0>
     }]
    relativeSpaceBone {boneName=root}
   }
   n53 = Anim Dynamics - Bone: dyn_l_shoulder_leaf_05(ComponentPose=@n52, Alpha=@n39, ExternalForce="0,0,0") {
    angularDampingOverride=200
    angularSpringConstant=500
    bAngularSpring=true
    bGravityOverrideInSimSpace=true
    bUseGravityOverride=true
    simulationSpace=BoneRelative
    alphaBoolBlend {blendOption=Linear}
    boundBone {boneName=dyn_l_shoulder_leaf_05}
    componentAppliedLinearAccClamp=<1500 1500 1500>
    componentLinearAccScale=<0.4 0.4 1>
    componentLinearVelScale=<0.2 0.2 0.4>
    gravityOverride {z=-980}
    physicsBodyDefinitions [{
      sphereCollisionRadius=0
      boundBone {boneName=dyn_l_shoulder_leaf_05}
      boxExtents=<7 7 7>
      constraintSetup {angularLimitsMax=<8 5 8>, angularLimitsMin=<-8 -5 -8>, angularTarget=<1 0 0>}
      localJointOffset=<0.2 0 0>
     }]
    relativeSpaceBone {boneName=root}
   }
   n54 = Anim Dynamics - Bone: dyn_l_shoulder_leaf_06(ComponentPose=@n53, Alpha=@n39, ExternalForce="0,0,0") {
    angularDampingOverride=200
    angularSpringConstant=500
    bAngularSpring=true
    bGravityOverrideInSimSpace=true
    bUseGravityOverride=true
    simulationSpace=BoneRelative
    alphaBoolBlend {blendOption=Linear}
    boundBone {boneName=dyn_l_shoulder_leaf_06}
    componentAppliedLinearAccClamp=<1500 1500 1500>
    componentLinearAccScale=<0.4 0.4 1>
    componentLinearVelScale=<0.2 0.2 0.4>
    gravityOverride {z=-980}
    physicsBodyDefinitions [{
      sphereCollisionRadius=0
      boundBone {boneName=dyn_l_shoulder_leaf_06}
      boxExtents=<7 7 7>
      constraintSetup {angularLimitsMax=<8 5 8>, angularLimitsMin=<-8 -5 -8>, angularTarget=<1 0 0>}
      localJointOffset=<0.2 0 0>
     }]
    relativeSpaceBone {boneName=root}
   }
   n55 = Anim Dynamics - Bone: dyn_l_shoulder_leaf_07(ComponentPose=@n54, Alpha=@n39, ExternalForce="0,0,0") {
    angularDampingOverride=200
    angularSpringConstant=500
    bAngularSpring=true
    bGravityOverrideInSimSpace=true
    bUseGravityOverride=true
    simulationSpace=BoneRelative
    alphaBoolBlend {blendOption=Linear}
    boundBone {boneName=dyn_l_shoulder_leaf_07}
    componentAppliedLinearAccClamp=<1500 1500 1500>
    componentLinearAccScale=<0.4 0.4 1>
    componentLinearVelScale=<0.2 0.2 0.4>
    gravityOverride {z=-980}
    physicsBodyDefinitions [{
      sphereCollisionRadius=0
      boundBone {boneName=dyn_l_shoulder_leaf_07}
      boxExtents=<7 7 7>
      constraintSetup {angularLimitsMax=<8 -5 8>, angularLimitsMin=<-8 -15 -8>, angularTarget=<1 0 0>}
      localJointOffset=<0.2 0 0>
     }]
    relativeSpaceBone {boneName=root}
   }
   n56 = Anim Dynamics - Bone: dyn_l_shoulder_leaf_08(ComponentPose=@n55, Alpha=@n39, ExternalForce="0,0,0") {
    angularDampingOverride=200
    angularSpringConstant=500
    bAngularSpring=true
    bGravityOverrideInSimSpace=true
    bUseGravityOverride=true
    simulationSpace=BoneRelative
    alphaBoolBlend {blendOption=Linear}
    boundBone {boneName=dyn_l_shoulder_leaf_08}
    componentAppliedLinearAccClamp=<1500 1500 1500>
    componentLinearAccScale=<0.4 0.4 1>
    componentLinearVelScale=<0.2 0.2 0.4>
    gravityOverride {z=-980}
    physicsBodyDefinitions [{
      sphereCollisionRadius=0
      boundBone {boneName=dyn_l_shoulder_leaf_08}
      boxExtents=<7 7 7>
      constraintSetup {angularLimitsMax=<8 5 8>, angularLimitsMin=<-8 -5 -8>, angularTarget=<1 0 0>}
      localJointOffset=<0.2 0 0>
     }]
    relativeSpaceBone {boneName=root}
   }
   n57 = Anim Dynamics - Bone: dyn_l_shoulder_leaf_09(ComponentPose=@n56, Alpha=@n39, ExternalForce="0,0,0") {
    angularDampingOverride=200
    angularSpringConstant=500
    bAngularSpring=true
    bGravityOverrideInSimSpace=true
    bUseGravityOverride=true
    simulationSpace=BoneRelative
    alphaBoolBlend {blendOption=Linear}
    boundBone {boneName=dyn_l_shoulder_leaf_09}
    componentAppliedLinearAccClamp=<1500 1500 1500>
    componentLinearAccScale=<0.4 0.4 1>
    componentLinearVelScale=<0.2 0.2 0.4>
    gravityOverride {z=-980}
    physicsBodyDefinitions [{
      sphereCollisionRadius=0
      boundBone {boneName=dyn_l_shoulder_leaf_09}
      boxExtents=<7 7 7>
      constraintSetup {angularLimitsMax=<8 5 8>, angularLimitsMin=<-8 -5 -8>, angularTarget=<1 0 0>}
      localJointOffset=<0.2 0 0>
     }]
    relativeSpaceBone {boneName=root}
   }
   n58 = Anim Dynamics - Bone: dyn_l_shoulder_leaf_010(ComponentPose=@n57, Alpha=@n39, ExternalForce="0,0,0") {
    angularDampingOverride=200
    angularSpringConstant=500
    bAngularSpring=true
    bGravityOverrideInSimSpace=true
    bUseGravityOverride=true
    simulationSpace=BoneRelative
    alphaBoolBlend {blendOption=Linear}
    boundBone {boneName=dyn_l_shoulder_leaf_010}
    componentAppliedLinearAccClamp=<1500 1500 1500>
    componentLinearAccScale=<0.4 0.4 1>
    componentLinearVelScale=<0.2 0.2 0.4>
    gravityOverride {z=-980}
    physicsBodyDefinitions [{
      sphereCollisionRadius=0
      boundBone {boneName=dyn_l_shoulder_leaf_010}
      boxExtents=<7 7 7>
      constraintSetup {angularLimitsMax=<8 5 8>, angularLimitsMin=<-8 -5 -8>, angularTarget=<1 0 0>}
      localJointOffset=<0.2 0 0>
     }]
    relativeSpaceBone {boneName=root}
   }
   n59 = Anim Dynamics - Bone: dyn_l_shoulder_leaf_011(ComponentPose=@n58, Alpha=@n39, ExternalForce="0,0,0") {
    angularDampingOverride=200
    angularSpringConstant=500
    bAngularSpring=true
    bGravityOverrideInSimSpace=true
    bUseGravityOverride=true
    simulationSpace=BoneRelative
    alphaBoolBlend {blendOption=Linear}
    boundBone {boneName=dyn_l_shoulder_leaf_011}
    componentAppliedLinearAccClamp=<1500 1500 1500>
    componentLinearAccScale=<0.4 0.4 1>
    componentLinearVelScale=<0.2 0.2 0.4>
    gravityOverride {z=-980}
    physicsBodyDefinitions [{
      sphereCollisionRadius=0
      boundBone {boneName=dyn_l_shoulder_leaf_011}
      boxExtents=<7 7 7>
      constraintSetup {angularLimitsMax=<8 5 8>, angularLimitsMin=<-8 -5 -8>, angularTarget=<1 0 0>}
      localJointOffset=<0.2 0 0>
     }]
    relativeSpaceBone {boneName=root}
   }
   n60 = Anim Dynamics - Bone: dyn_l_shoulder_leaf_012(ComponentPose=@n59, Alpha=@n39, ExternalForce="0,0,0") {
    angularDampingOverride=200
    angularSpringConstant=500
    bAngularSpring=true
    bGravityOverrideInSimSpace=true
    bUseGravityOverride=true
    simulationSpace=BoneRelative
    alphaBoolBlend {blendOption=Linear}
    boundBone {boneName=dyn_l_shoulder_leaf_012}
    componentAppliedLinearAccClamp=<1500 1500 1500>
    componentLinearAccScale=<0.4 0.4 1>
    componentLinearVelScale=<0.2 0.2 0.4>
    gravityOverride {z=-980}
    physicsBodyDefinitions [{
      sphereCollisionRadius=0
      boundBone {boneName=dyn_l_shoulder_leaf_012}
      boxExtents=<7 7 7>
      constraintSetup {angularLimitsMax=<8 0 8>, angularLimitsMin=<-8 -8 -8>, angularTarget=<1 0 0>}
      localJointOffset=<0.2 0 0>
     }]
    relativeSpaceBone {boneName=root}
   }
   n61 = Anim Dynamics - Bone: dyn_l_shoulder_leaf_013(ComponentPose=@n60, Alpha=@n39, ExternalForce="0,0,0") {
    angularDampingOverride=200
    angularSpringConstant=500
    bAngularSpring=true
    bGravityOverrideInSimSpace=true
    bUseGravityOverride=true
    simulationSpace=BoneRelative
    alphaBoolBlend {blendOption=Linear}
    boundBone {boneName=dyn_l_shoulder_leaf_013}
    componentAppliedLinearAccClamp=<1500 1500 1500>
    componentLinearAccScale=<0.4 0.4 1>
    componentLinearVelScale=<0.2 0.2 0.4>
    gravityOverride {z=-980}
    physicsBodyDefinitions [{
      sphereCollisionRadius=0
      boundBone {boneName=dyn_l_shoulder_leaf_013}
      boxExtents=<7 7 7>
      constraintSetup {angularLimitsMax=<8 5 8>, angularLimitsMin=<-8 -5 -8>, angularTarget=<1 0 0>}
      localJointOffset=<0.2 0 0>
     }]
    relativeSpaceBone {boneName=root}
   }
   n62 = Anim Dynamics - Bone: dyn_l_shoulder_leaf_014(ComponentPose=@n61, Alpha=@n39, ExternalForce="0,0,0") {
    angularDampingOverride=200
    angularSpringConstant=500
    bAngularSpring=true
    bGravityOverrideInSimSpace=true
    bUseGravityOverride=true
    simulationSpace=BoneRelative
    alphaBoolBlend {blendOption=Linear}
    boundBone {boneName=dyn_l_shoulder_leaf_014}
    componentAppliedLinearAccClamp=<1500 1500 1500>
    componentLinearAccScale=<0.4 0.4 1>
    componentLinearVelScale=<0.2 0.2 0.4>
    gravityOverride {z=-980}
    physicsBodyDefinitions [{
      sphereCollisionRadius=0
      boundBone {boneName=dyn_l_shoulder_leaf_014}
      boxExtents=<7 7 7>
      constraintSetup {angularLimitsMax=<8 5 8>, angularLimitsMin=<-8 -5 -8>, angularTarget=<1 0 0>}
      localJointOffset=<0.2 0 0>
     }]
    relativeSpaceBone {boneName=root}
   }
   n63 = Anim Dynamics - Bone: dyn_l_shoulder_leaf_015(ComponentPose=@n62, Alpha=@n39, ExternalForce="0,0,0") {
    angularDampingOverride=200
    angularSpringConstant=500
    bAngularSpring=true
    bGravityOverrideInSimSpace=true
    bUseGravityOverride=true
    simulationSpace=BoneRelative
    alphaBoolBlend {blendOption=Linear}
    boundBone {boneName=dyn_l_shoulder_leaf_015}
    componentAppliedLinearAccClamp=<1500 1500 1500>
    componentLinearAccScale=<0.4 0.4 1>
    componentLinearVelScale=<0.2 0.2 0.4>
    gravityOverride {z=-980}
    physicsBodyDefinitions [{
      sphereCollisionRadius=0
      boundBone {boneName=dyn_l_shoulder_leaf_015}
      boxExtents=<7 7 7>
      constraintSetup {angularLimitsMax=<8 5 8>, angularLimitsMin=<-8 -5 -8>, angularTarget=<1 0 0>}
      localJointOffset=<0.2 0 0>
     }]
    relativeSpaceBone {boneName=root}
   }
   n64 = Anim Dynamics - Bone: dyn_l_arm_flower_01(ComponentPose=@n63, Alpha=@n39, ExternalForce="0,0,0") {
    angularDampingOverride=200
    angularSpringConstant=500
    bAngularSpring=true
    bGravityOverrideInSimSpace=true
    bUseGravityOverride=true
    simulationSpace=BoneRelative
    alphaBoolBlend {blendOption=Linear}
    boundBone {boneName=dyn_l_arm_flower_01}
    componentAppliedLinearAccClamp=<1500 1500 1500>
    componentLinearAccScale=<0.4 0.4 1>
    componentLinearVelScale=<0.2 0.2 0.4>
    gravityOverride {z=-980}
    physicsBodyDefinitions [{
      sphereCollisionRadius=0
      boundBone {boneName=dyn_l_arm_flower_01}
      boxExtents=<6 6 6>
      constraintSetup {angularLimitsMax=<8 8 8>, angularLimitsMin=<-8 -8 -8>, angularTarget=<1 0 0>}
      localJointOffset=<0.2 0 0>
     }]
    relativeSpaceBone {boneName=root}
   }
   n65 = Anim Dynamics - Bone: dyn_l_arm_flower_02(ComponentPose=@n64, Alpha=@n39, ExternalForce="0,0,0") {
    angularDampingOverride=200
    angularSpringConstant=500
    bAngularSpring=true
    bGravityOverrideInSimSpace=true
    bUseGravityOverride=true
    simulationSpace=BoneRelative
    alphaBoolBlend {blendOption=Linear}
    boundBone {boneName=dyn_l_arm_flower_02}
    componentAppliedLinearAccClamp=<1500 1500 1500>
    componentLinearAccScale=<0.4 0.4 1>
    componentLinearVelScale=<0.2 0.2 0.4>
    gravityOverride {z=-980}
    physicsBodyDefinitions [{
      sphereCollisionRadius=0
      boundBone {boneName=dyn_l_arm_flower_02}
      boxExtents=<6 6 6>
      constraintSetup {angularLimitsMax=<8 8 8>, angularLimitsMin=<-8 -8 -8>, angularTarget=<1 0 0>}
      localJointOffset=<0.2 0 0>
     }]
    relativeSpaceBone {boneName=root}
   }
   n66 = Anim Dynamics - Bone: dyn_l_arm_flower_03(ComponentPose=@n65, Alpha=@n39, ExternalForce="0,0,0") {
    angularDampingOverride=200
    angularSpringConstant=500
    bAngularSpring=true
    bGravityOverrideInSimSpace=true
    bUseGravityOverride=true
    simulationSpace=BoneRelative
    alphaBoolBlend {blendOption=Linear}
    boundBone {boneName=dyn_l_arm_flower_03}
    componentAppliedLinearAccClamp=<1500 1500 1500>
    componentLinearAccScale=<0.4 0.4 1>
    componentLinearVelScale=<0.2 0.2 0.4>
    gravityOverride {z=-980}
    physicsBodyDefinitions [{
      sphereCollisionRadius=0
      boundBone {boneName=dyn_l_arm_flower_03}
      boxExtents=<6 6 6>
      constraintSetup {angularLimitsMax=<8 8 8>, angularLimitsMin=<-8 -8 -8>, angularTarget=<1 0 0>}
      localJointOffset=<0.2 0 0>
     }]
    relativeSpaceBone {boneName=root}
   }
   n67 = "Constraint - dyn_r_shoulder_leaf_01_master (PP)"!(ComponentPose=@n66, Alpha=1, ConstraintWeights_0=0.3098, ConstraintWeights_1:float=0.6902) -> (Pose) {
    alphaBoolBlend {blendOption=Linear}
    boneToModify {boneName=dyn_r_shoulder_leaf_01_master}
    constraintSetup [{targetBone {boneName=upperarm_twist_01_r}}, {targetBone {boneName=clavicle_r}}]
    constraintWeights [1,1]
   }
   n68 = "Constraint - dyn_r_shoulder_flower_03_master (PPP)"!(ComponentPose=@n67, Alpha=1, ConstraintWeights_0=0.3408, ConstraintWeights_1:float=0.6592, ConstraintWeights_2:float=0) -> (Pose) {
    alphaBoolBlend {blendOption=Linear}
    boneToModify {boneName=dyn_r_shoulder_flower_03_master}
    constraintSetup [{targetBone {boneName=upperarm_twist_01_r}}, {targetBone {boneName=clavicle_r}}, {targetBone {boneName=spine_03}}]
    constraintWeights [1,1,1]
   }
   n69 = "Constraint - dyn_r_shoulder_leaf_02_master (PP)"!(ComponentPose=@n68, Alpha=1, ConstraintWeights_0=0.5044, ConstraintWeights_1:float=0.4956) -> (Pose) {
    alphaBoolBlend {blendOption=Linear}
    boneToModify {boneName=dyn_r_shoulder_leaf_02_master}
    constraintSetup [{targetBone {boneName=upperarm_twist_01_r}}, {targetBone {boneName=clavicle_r}}]
    constraintWeights [1,1]
   }
   n70 = "Constraint - dyn_r_shoulder_leaf_03_master (PPP)"!(ComponentPose=@n69, Alpha=1, ConstraintWeights_0=0.6287, ConstraintWeights_1:float=0.3713, ConstraintWeights_2:float=0) -> (Pose) {
    alphaBoolBlend {blendOption=Linear}
    boneToModify {boneName=dyn_r_shoulder_leaf_03_master}
    constraintSetup [{targetBone {boneName=upperarm_twist_01_r}}, {targetBone {boneName=clavicle_r}}, {targetBone {boneName=spine_03}}]
    constraintWeights [1,1,1]
   }
   n71 = "Constraint - dyn_r_shoulder_leaf_04_master (PPP)"!(ComponentPose=@n70, Alpha=1, ConstraintWeights_0=0.7621, ConstraintWeights_1:float=0.2379, ConstraintWeights_2:float=0) -> (Pose) {
    alphaBoolBlend {blendOption=Linear}
    boneToModify {boneName=dyn_r_shoulder_leaf_04_master}
    constraintSetup [{targetBone {boneName=upperarm_twist_01_r}}, {targetBone {boneName=clavicle_r}}, {targetBone {boneName=spine_03}}]
    constraintWeights [1,1,1]
   }
   n72 = "Constraint - dyn_r_shoulder_leaf_05_master (PPP)"!(ComponentPose=@n71, Alpha=1, ConstraintWeights_0=0.7, ConstraintWeights_1:float=0.3, ConstraintWeights_2:float=0) -> (Pose) {
    alphaBoolBlend {blendOption=Linear}
    boneToModify {boneName=dyn_r_shoulder_leaf_05_master}
    constraintSetup [{targetBone {boneName=upperarm_twist_01_r}}, {targetBone {boneName=clavicle_r}}, {targetBone {boneName=spine_03}}]
    constraintWeights [1,1,1]
   }
   n73 = "Constraint - dyn_r_shoulder_leaf_06_master (PPP)"!(ComponentPose=@n72, Alpha=1, ConstraintWeights_0=0.3526, ConstraintWeights_1:float=0.6474, ConstraintWeights_2:float=0) -> (Pose) {
    alphaBoolBlend {blendOption=Linear}
    boneToModify {boneName=dyn_r_shoulder_leaf_06_master}
    constraintSetup [{targetBone {boneName=upperarm_twist_01_r}}, {targetBone {boneName=clavicle_r}}, {targetBone {boneName=spine_03}}]
    constraintWeights [1,1,1]
   }
   n74 = "Constraint - dyn_r_shoulder_leaf_07_master (PPPPP)"!(
    ComponentPose=@n73
    Alpha=1
    ConstraintWeights_0=0.7644
    ConstraintWeights_1:float=0.0871
    ConstraintWeights_2:float=0.0517
    ConstraintWeights_3:float=0
    ConstraintWeights_4:float=0.0969
   ) -> (Pose) {
    alphaBoolBlend {blendOption=Linear}
    boneToModify {boneName=dyn_r_shoulder_leaf_07_master}
    constraintSetup [
     {targetBone {boneName=upperarm_twist_01_r}}
     {targetBone {boneName=clavicle_r}}
     {targetBone {boneName=upperarm_r}}
     {targetBone {boneName=spine_03}}
     {targetBone {boneName=upperarm_twist_02_r}}
    ]
    constraintWeights [1,1,1,1,1]
   }
   n75 = "Constraint - dyn_r_shoulder_leaf_08_master (PPPP)"!(
    ComponentPose=@n74
    Alpha=1
    ConstraintWeights_0=0.8171
    ConstraintWeights_1:float=0.1335
    ConstraintWeights_2:float=0.0348
    ConstraintWeights_3:float=0.0146
   ) -> (Pose) {
    alphaBoolBlend {blendOption=Linear}
    boneToModify {boneName=dyn_r_shoulder_leaf_08_master}
    constraintSetup [
     {targetBone {boneName=upperarm_twist_01_r}}
     {targetBone {boneName=clavicle_r}}
     {targetBone {boneName=upperarm_r}}
     {targetBone {boneName=upperarm_twist_02_r}}
    ]
    constraintWeights [1,1,1,1]
   }
   n76 = "Constraint - dyn_r_shoulder_leaf_09_master (PPPP)"!(
    ComponentPose=@n75
    Alpha=1
    ConstraintWeights_0=0.8201
    ConstraintWeights_1:float=0.0585
    ConstraintWeights_2:float=0.0716
    ConstraintWeights_3:float=0.0498
   ) -> (Pose) {
    alphaBoolBlend {blendOption=Linear}
    boneToModify {boneName=dyn_r_shoulder_leaf_09_master}
    constraintSetup [
     {targetBone {boneName=upperarm_twist_01_r}}
     {targetBone {boneName=clavicle_r}}
     {targetBone {boneName=upperarm_r}}
     {targetBone {boneName=upperarm_twist_02_r}}
    ]
    constraintWeights [1,1,1,1]
   }
   n77 = "Constraint - dyn_r_shoulder_leaf_10_master (PPPP)"!(
    ComponentPose=@n76
    Alpha=1
    ConstraintWeights_0=0.8153
    ConstraintWeights_1:float=0.1695
    ConstraintWeights_2:float=0.0153
    ConstraintWeights_3:float=0
   ) -> (Pose) {
    alphaBoolBlend {blendOption=Linear}
    boneToModify {boneName=dyn_r_shoulder_leaf_10_master}
    constraintSetup [
     {targetBone {boneName=upperarm_twist_01_r}}
     {targetBone {boneName=clavicle_r}}
     {targetBone {boneName=upperarm_r}}
     {targetBone {boneName=upperarm_twist_02_r}}
    ]
    constraintWeights [1,1,1,1]
   }
   n78 = "Constraint - dyn_r_shoulder_leaf_11_master (PPP)"!(ComponentPose=@n77, Alpha=1, ConstraintWeights_0=0.8315, ConstraintWeights_1:float=0.1338, ConstraintWeights_2:float=0.0347) -> (Pose) {
    alphaBoolBlend {blendOption=Linear}
    boneToModify {boneName=dyn_r_shoulder_leaf_11_master}
    constraintSetup [{targetBone {boneName=upperarm_twist_01_r}}, {targetBone {boneName=clavicle_r}}, {targetBone {boneName=upperarm_r}}]
    constraintWeights [1,1,1]
   }
   n79 = "Constraint - dyn_r_shoulder_leaf_12_master (PP)"!(ComponentPose=@n78, Alpha=1, ConstraintWeights_0=0.6589, ConstraintWeights_1:float=0.3411) -> (Pose) {
    alphaBoolBlend {blendOption=Linear}
    boneToModify {boneName=dyn_r_shoulder_leaf_12_master}
    constraintSetup [{targetBone {boneName=upperarm_twist_01_r}}, {targetBone {boneName=clavicle_r}}]
    constraintWeights [1,1]
   }
   n80 = "Constraint - dyn_r_shoulder_leaf_13_master (PPP)"!(ComponentPose=@n79, Alpha=1, ConstraintWeights_0=0.5527, ConstraintWeights_1:float=0.4473, ConstraintWeights_2:float=0) -> (Pose) {
    alphaBoolBlend {blendOption=Linear}
    boneToModify {boneName=dyn_r_shoulder_leaf_13_master}
    constraintSetup [{targetBone {boneName=upperarm_twist_01_r}}, {targetBone {boneName=clavicle_r}}, {targetBone {boneName=spine_03}}]
    constraintWeights [1,1,1]
   }
   n81 = "Constraint - dyn_r_shoulder_leaf_14_master (PPP)"!(ComponentPose=@n80, Alpha=1, ConstraintWeights_0=0.576, ConstraintWeights_1:float=0.424, ConstraintWeights_2:float=0) -> (Pose) {
    alphaBoolBlend {blendOption=Linear}
    boneToModify {boneName=dyn_r_shoulder_leaf_14_master}
    constraintSetup [{targetBone {boneName=upperarm_twist_01_r}}, {targetBone {boneName=clavicle_r}}, {targetBone {boneName=spine_03}}]
    constraintWeights [1,1,1]
   }
   n82 = "Constraint - dyn_r_shoulder_leaf_15_master (PPP)"!(ComponentPose=@n81, Alpha=1, ConstraintWeights_0=0.3912, ConstraintWeights_1:float=0.6088, ConstraintWeights_2:float=0) -> (Pose) {
    alphaBoolBlend {blendOption=Linear}
    boneToModify {boneName=dyn_r_shoulder_leaf_15_master}
    constraintSetup [{targetBone {boneName=upperarm_twist_01_r}}, {targetBone {boneName=clavicle_r}}, {targetBone {boneName=spine_03}}]
    constraintWeights [1,1,1]
   }
   n83 = "Constraint - dyn_r_shoulder_leaf_16_master (PP)"!(ComponentPose=@n82, Alpha=1, ConstraintWeights_0=0.3677, ConstraintWeights_1:float=0.6323) -> (Pose) {
    alphaBoolBlend {blendOption=Linear}
    boneToModify {boneName=dyn_r_shoulder_leaf_16_master}
    constraintSetup [{targetBone {boneName=upperarm_twist_01_r}}, {targetBone {boneName=clavicle_r}}]
    constraintWeights [1,1]
   }
   n84 = "Constraint - dyn_r_shoulder_leaf_17_master (PP)"!(ComponentPose=@n83, Alpha=1, ConstraintWeights_0=0.6653, ConstraintWeights_1:float=0.3347) -> (Pose) {
    alphaBoolBlend {blendOption=Linear}
    boneToModify {boneName=dyn_r_shoulder_leaf_17_master}
    constraintSetup [{targetBone {boneName=upperarm_twist_01_r}}, {targetBone {boneName=clavicle_r}}]
    constraintWeights [1,1]
   }
   n85 = "Constraint - dyn_r_shoulder_leaf_18_master (PPP)"!(ComponentPose=@n84, Alpha=1, ConstraintWeights_0=0.6168, ConstraintWeights_1:float=0.3832, ConstraintWeights_2:float=0) -> (Pose) {
    alphaBoolBlend {blendOption=Linear}
    boneToModify {boneName=dyn_r_shoulder_leaf_18_master}
    constraintSetup [{targetBone {boneName=upperarm_twist_01_r}}, {targetBone {boneName=clavicle_r}}, {targetBone {boneName=spine_03}}]
    constraintWeights [1,1,1]
   }
   n86 = "Constraint - dyn_r_shoulder_flower_01_master (PPP)"!(ComponentPose=@n85, Alpha=1, ConstraintWeights_0=0.6291, ConstraintWeights_1:float=0.3709, ConstraintWeights_2:float=0) -> (Pose) {
    alphaBoolBlend {blendOption=Linear}
    boneToModify {boneName=dyn_r_shoulder_flower_01_master}
    constraintSetup [{targetBone {boneName=upperarm_twist_01_r}}, {targetBone {boneName=clavicle_r}}, {targetBone {boneName=spine_03}}]
    constraintWeights [1,1,1]
   }
   n87 = "Constraint - dyn_r_shoulder_flower_02_master (PPPP)"!(
    ComponentPose=@n86
    Alpha=1
    ConstraintWeights_0=0.8417
    ConstraintWeights_1:float=0.0679
    ConstraintWeights_2:float=0.0661
    ConstraintWeights_3:float=0.0243
   ) -> (Pose) {
    alphaBoolBlend {blendOption=Linear}
    boneToModify {boneName=dyn_r_shoulder_flower_02_master}
    constraintSetup [
     {targetBone {boneName=upperarm_twist_01_r}}
     {targetBone {boneName=clavicle_r}}
     {targetBone {boneName=upperarm_r}}
     {targetBone {boneName=upperarm_twist_02_r}}
    ]
    constraintWeights [1,1,1,1]
   }
   n88 = "Constraint - dyn_r_shoulder_flower_03_master (PP)"!(ComponentPose=@n87, Alpha=1, ConstraintWeights_0=0.7578, ConstraintWeights_1:float=0.2422) -> (Pose) {
    alphaBoolBlend {blendOption=Linear}
    boneToModify {boneName=dyn_r_shoulder_flower_03_master}
    constraintSetup [{targetBone {boneName=upperarm_twist_01_r}}, {targetBone {boneName=clavicle_r}}]
    constraintWeights [1,1]
   }
   n89 = "Constraint - dyn_r_shoulder_flower_04_master (PPPP)"!(
    ComponentPose=@n88
    Alpha=1
    ConstraintWeights_0=0.523
    ConstraintWeights_1:float=0.477
    ConstraintWeights_2:float=0
    ConstraintWeights_3:float=0
   ) -> (Pose) {
    alphaBoolBlend {blendOption=Linear}
    boneToModify {boneName=dyn_r_shoulder_flower_04_master}
    constraintSetup [
     {targetBone {boneName=upperarm_twist_01_r}}
     {targetBone {boneName=clavicle_r}}
     {targetBone {boneName=spine_02}}
     {targetBone {boneName=spine_03}}
    ]
    constraintWeights [1,1,1,1]
   }
   n90 = "Constraint - dyn_r_shoulder_flower_05_master (PPP)"!(ComponentPose=@n89, Alpha=1, ConstraintWeights_0=0.3327, ConstraintWeights_1:float=0.6673, ConstraintWeights_2:float=0) -> (Pose) {
    alphaBoolBlend {blendOption=Linear}
    boneToModify {boneName=dyn_r_shoulder_flower_05_master}
    constraintSetup [{targetBone {boneName=upperarm_twist_01_r}}, {targetBone {boneName=clavicle_r}}, {targetBone {boneName=spine_03}}]
    constraintWeights [1,1,1]
   }
   n91 = "Constraint - dyn_r_arm_leaf_01_master (PPPP)"!(
    ComponentPose=@n90
    Alpha=1
    ConstraintWeights_0=0.0449
    ConstraintWeights_1:float=0.5645
    ConstraintWeights_2:float=0.318
    ConstraintWeights_3:float=0.0725
   ) -> (Pose) {
    alphaBoolBlend {blendOption=Linear}
    boneToModify {boneName=dyn_r_arm_leaf_01_master}
    constraintSetup [
     {targetBone {boneName=upperarm_twist_01_r}}
     {targetBone {boneName=upperarm_r}}
     {targetBone {boneName=lowerarm_r}}
     {targetBone {boneName=upperarm_twist_02_r}}
    ]
    constraintWeights [1,1,1,1]
   }
   n92 = "Constraint - dyn_r_arm_leaf_02_master (PPPPP)"!(
    ComponentPose=@n91
    Alpha=1
    ConstraintWeights_0=0.125
    ConstraintWeights_1:float=0.0223
    ConstraintWeights_2:float=0.3885
    ConstraintWeights_3:float=0.4254
    ConstraintWeights_4:float=0.0388
   ) -> (Pose) {
    alphaBoolBlend {blendOption=Linear}
    boneToModify {boneName=dyn_r_arm_leaf_02_master}
    constraintSetup [
     {targetBone {boneName=lowerarm_twist_02_r}}
     {targetBone {boneName=upperarm_twist_01_r}}
     {targetBone {boneName=upperarm_r}}
     {targetBone {boneName=lowerarm_r}}
     {targetBone {boneName=upperarm_twist_02_r}}
    ]
    constraintWeights [1,1,1,1,1]
   }
   n93 = "Constraint - dyn_r_arm_leaf_03_master (PPPPP)"!(
    ComponentPose=@n92
    Alpha=1
    ConstraintWeights_0=0.2284
    ConstraintWeights_1:float=0.0139
    ConstraintWeights_2:float=0.2857
    ConstraintWeights_3:float=0.4454
    ConstraintWeights_4:float=0.0267
   ) -> (Pose) {
    alphaBoolBlend {blendOption=Linear}
    boneToModify {boneName=dyn_r_arm_leaf_03_master}
    constraintSetup [
     {targetBone {boneName=lowerarm_twist_02_r}}
     {targetBone {boneName=upperarm_twist_01_r}}
     {targetBone {boneName=upperarm_r}}
     {targetBone {boneName=lowerarm_r}}
     {targetBone {boneName=upperarm_twist_02_r}}
    ]
    constraintWeights [1,1,1,1,1]
   }
   n94 = "Constraint - dyn_r_arm_leaf_04_master (P)"!(ComponentPose=@n93, Alpha=1, ConstraintWeights_0=1, ConstraintWeights_1:float=0) -> (Pose) {
    alphaBoolBlend {blendOption=Linear}
    boneToModify {boneName=dyn_r_arm_leaf_04_master}
    constraintSetup [{targetBone {boneName=lowerarm_twist_01_r}}]
    constraintWeights [1,0]
   }
   n95 = "Constraint - dyn_r_hand_flower_01_master (PP)"!(ComponentPose=@n94, Alpha=1, ConstraintWeights_0=0.2418, ConstraintWeights_1:float=0.7582) -> (Pose) {
    alphaBoolBlend {blendOption=Linear}
    boneToModify {boneName=dyn_r_hand_flower_01_master}
    constraintSetup [{targetBone {boneName=middle_metacarpal_r}}, {targetBone {boneName=index_metacarpal_r}}]
    constraintWeights [1,1]
   }
   n96 = "Constraint - dyn_r_hand_flower_02_master (PPP)"!(ComponentPose=@n95, Alpha=1, ConstraintWeights_0=0.0569, ConstraintWeights_1:float=0.9321, ConstraintWeights_2:float=0.011) -> (Pose) {
    alphaBoolBlend {blendOption=Linear}
    boneToModify {boneName=dyn_r_hand_flower_02_master}
    constraintSetup [{targetBone {boneName=middle_metacarpal_r}}, {targetBone {boneName=ring_metacarpal_r}}, {targetBone {boneName=pinky_metacarpal_r}}]
    constraintWeights [1,1,1]
   }
   n97 = "Constraint - dyn_l_shoulder_flower_01_master (PPP)"!(ComponentPose=@n96, Alpha=1, ConstraintWeights_0=0.0137, ConstraintWeights_1:float=0.5783, ConstraintWeights_2:float=0.4079) -> (Pose) {
    alphaBoolBlend {blendOption=Linear}
    boneToModify {boneName=dyn_l_shoulder_flower_01_master}
    constraintSetup [{targetBone {boneName=spine_03}}, {targetBone {boneName=upperarm_twist_01_l}}, {targetBone {boneName=clavicle_l}}]
    constraintWeights [1,1,1]
   }
   n98 = "Constraint - dyn_l_shoulder_flower_02_master (PP)"!(ComponentPose=@n97, Alpha=1, ConstraintWeights_0=0.6457, ConstraintWeights_1:float=0.3543) -> (Pose) {
    alphaBoolBlend {blendOption=Linear}
    boneToModify {boneName=dyn_l_shoulder_flower_02_master}
    constraintSetup [{targetBone {boneName=upperarm_twist_01_l}}, {targetBone {boneName=clavicle_l}}]
    constraintWeights [1,1]
   }
   n99 = "Constraint - dyn_l_shoulder_flower_03_master (PP)"!(ComponentPose=@n98, Alpha=1, ConstraintWeights_0=0.4084, ConstraintWeights_1:float=0.5916) -> (Pose) {
    alphaBoolBlend {blendOption=Linear}
    boneToModify {boneName=dyn_l_shoulder_flower_03_master}
    constraintSetup [{targetBone {boneName=upperarm_twist_01_l}}, {targetBone {boneName=clavicle_l}}]
    constraintWeights [1,1]
   }
   n100 = "Constraint - dyn_l_shoulder_flower_04_master (PP)"!(ComponentPose=@n99, Alpha=1, ConstraintWeights_0=0.3014, ConstraintWeights_1:float=0.6986) -> (Pose) {
    alphaBoolBlend {blendOption=Linear}
    boneToModify {boneName=dyn_l_shoulder_flower_04_master}
    constraintSetup [{targetBone {boneName=upperarm_twist_01_l}}, {targetBone {boneName=clavicle_l}}]
    constraintWeights [1,1]
   }
   n101 = "Constraint - dyn_l_shoulder_flower_06_master (PPP)"!(ComponentPose=@n100, Alpha=1, ConstraintWeights_0=0.0049, ConstraintWeights_1:float=0.7914, ConstraintWeights_2:float=0.2037) -> (Pose) {
    alphaBoolBlend {blendOption=Linear}
    boneToModify {boneName=dyn_l_shoulder_flower_06_master}
    constraintSetup [{targetBone {boneName=spine_03}}, {targetBone {boneName=upperarm_twist_01_l}}, {targetBone {boneName=clavicle_l}}]
    constraintWeights [1,1,1]
   }
   n102 = "Constraint - dyn_l_shoulder_flower_09_master (PP)"!(ComponentPose=@n101, Alpha=1, ConstraintWeights_0=0.5972, ConstraintWeights_1:float=0.4028) -> (Pose) {
    alphaBoolBlend {blendOption=Linear}
    boneToModify {boneName=dyn_l_shoulder_flower_09_master}
    constraintSetup [{targetBone {boneName=upperarm_twist_01_l}}, {targetBone {boneName=clavicle_l}}]
    constraintWeights [1,1]
   }
   n103 = "Constraint - dyn_l_shoulder_flower_08_master (PP)"!(ComponentPose=@n102, Alpha=1, ConstraintWeights_0=0.6432, ConstraintWeights_1:float=0.3568) -> (Pose) {
    alphaBoolBlend {blendOption=Linear}
    boneToModify {boneName=dyn_l_shoulder_flower_08_master}
    constraintSetup [{targetBone {boneName=upperarm_twist_01_l}}, {targetBone {boneName=clavicle_l}}]
    constraintWeights [1,1]
   }
   n104 = "Constraint - dyn_l_shoulder_flower_07_master (PPP)"!(ComponentPose=@n103, Alpha=1, ConstraintWeights_0=0.8493, ConstraintWeights_1:float=0.0979, ConstraintWeights_2:float=0.0528) -> (Pose) {
    alphaBoolBlend {blendOption=Linear}
    boneToModify {boneName=dyn_l_shoulder_flower_07_master}
    constraintSetup [{targetBone {boneName=upperarm_twist_01_l}}, {targetBone {boneName=clavicle_l}}, {targetBone {boneName=upperarm_l}}]
    constraintWeights [1,1,1]
   }
   n105 = "Constraint - dyn_l_arm_flower_01_master (PPP)"!(ComponentPose=@n104, Alpha=1, ConstraintWeights_0=0.4015, ConstraintWeights_1:float=0.3793, ConstraintWeights_2:float=0.2192) -> (Pose) {
    alphaBoolBlend {blendOption=Linear}
    boneToModify {boneName=dyn_l_arm_flower_01_master}
    constraintSetup [{targetBone {boneName=upperarm_twist_02_l}}, {targetBone {boneName=upperarm_l}}, {targetBone {boneName=upperarm_twist_01_l}}]
    constraintWeights [1,1,1]
   }
   n106 = "Constraint - dyn_l_shoulder_leaf_01_master (PP)"!(ComponentPose=@n105, Alpha=1, ConstraintWeights_0=0.7424, ConstraintWeights_1:float=0.2576) -> (Pose) {
    alphaBoolBlend {blendOption=Linear}
    boneToModify {boneName=dyn_l_shoulder_leaf_01_master}
    constraintSetup [{targetBone {boneName=upperarm_twist_01_l}}, {targetBone {boneName=clavicle_l}}]
    constraintWeights [1,1]
   }
   n107 = "Constraint - dyn_l_shoulder_leaf_02_master (PPPP)"!(
    ComponentPose=@n106
    Alpha=1
    ConstraintWeights_0=0.0307
    ConstraintWeights_1:float=0.8071
    ConstraintWeights_2:float=0.1256
    ConstraintWeights_3:float=0.0366
   ) -> (Pose) {
    alphaBoolBlend {blendOption=Linear}
    boneToModify {boneName=dyn_l_shoulder_leaf_02_master}
    constraintSetup [
     {targetBone {boneName=upperarm_twist_02_l}}
     {targetBone {boneName=upperarm_twist_01_l}}
     {targetBone {boneName=clavicle_l}}
     {targetBone {boneName=upperarm_l}}
    ]
    constraintWeights [1,1,1,1]
   }
   n108 = "Constraint - dyn_l_shoulder_leaf_03_master (PP)"!(ComponentPose=@n107, Alpha=1, ConstraintWeights_0=0.7106, ConstraintWeights_1:float=0.2894) -> (Pose) {
    alphaBoolBlend {blendOption=Linear}
    boneToModify {boneName=dyn_l_shoulder_leaf_03_master}
    constraintSetup [{targetBone {boneName=upperarm_twist_01_l}}, {targetBone {boneName=clavicle_l}}]
    constraintWeights [1,1]
   }
   n109 = "Constraint - dyn_l_shoulder_leaf_04_master (PPP)"!(ComponentPose=@n108, Alpha=1, ConstraintWeights_0=0.0039, ConstraintWeights_1:float=0.3529, ConstraintWeights_2:float=0.6432) -> (Pose) {
    alphaBoolBlend {blendOption=Linear}
    boneToModify {boneName=dyn_l_shoulder_leaf_04_master}
    constraintSetup [{targetBone {boneName=spine_03}}, {targetBone {boneName=upperarm_twist_01_l}}, {targetBone {boneName=clavicle_l}}]
    constraintWeights [1,1,1]
   }
   n110 = "Constraint - dyn_l_shoulder_leaf_05_master (PP)"!(ComponentPose=@n109, Alpha=1, ConstraintWeights_0=0.7736, ConstraintWeights_1:float=0.2264) -> (Pose) {
    alphaBoolBlend {blendOption=Linear}
    boneToModify {boneName=dyn_l_shoulder_leaf_05_master}
    constraintSetup [{targetBone {boneName=upperarm_twist_01_l}}, {targetBone {boneName=clavicle_l}}]
    constraintWeights [1,1]
   }
   n111 = "Constraint - dyn_l_shoulder_leaf_06_master (PP)"!(ComponentPose=@n110, Alpha=1, ConstraintWeights_0=0.3896, ConstraintWeights_1:float=0.6104) -> (Pose) {
    alphaBoolBlend {blendOption=Linear}
    boneToModify {boneName=dyn_l_shoulder_leaf_06_master}
    constraintSetup [{targetBone {boneName=upperarm_twist_01_l}}, {targetBone {boneName=clavicle_l}}]
    constraintWeights [1,1]
   }
   n112 = "Constraint - dyn_l_shoulder_leaf_07_master (PP)"!(ComponentPose=@n111, Alpha=1, ConstraintWeights_0=0.6097, ConstraintWeights_1:float=0.3903) -> (Pose) {
    alphaBoolBlend {blendOption=Linear}
    boneToModify {boneName=dyn_l_shoulder_leaf_07_master}
    constraintSetup [{targetBone {boneName=upperarm_twist_01_l}}, {targetBone {boneName=clavicle_l}}]
    constraintWeights [1,1]
   }
   n113 = "Constraint - dyn_l_shoulder_leaf_08_master (PP)"!(ComponentPose=@n112, Alpha=1, ConstraintWeights_0=0.5454, ConstraintWeights_1:float=0.4546) -> (Pose) {
    alphaBoolBlend {blendOption=Linear}
    boneToModify {boneName=dyn_l_shoulder_leaf_08_master}
    constraintSetup [{targetBone {boneName=upperarm_twist_01_l}}, {targetBone {boneName=clavicle_l}}]
    constraintWeights [1,1]
   }
   n114 = "Constraint - dyn_l_shoulder_leaf_09_master (PP)"!(ComponentPose=@n113, Alpha=1, ConstraintWeights_0=0.4679, ConstraintWeights_1:float=0.5321) -> (Pose) {
    alphaBoolBlend {blendOption=Linear}
    boneToModify {boneName=dyn_l_shoulder_leaf_09_master}
    constraintSetup [{targetBone {boneName=upperarm_twist_01_l}}, {targetBone {boneName=clavicle_l}}]
    constraintWeights [1,1]
   }
   n115 = "Constraint - dyn_l_shoulder_leaf_010_master (PPP)"!(ComponentPose=@n114, Alpha=1, ConstraintWeights_0=0.8348, ConstraintWeights_1:float=0.1281, ConstraintWeights_2:float=0.0371) -> (Pose) {
    alphaBoolBlend {blendOption=Linear}
    boneToModify {boneName=dyn_l_shoulder_leaf_010_master}
    constraintSetup [{targetBone {boneName=upperarm_twist_01_l}}, {targetBone {boneName=clavicle_l}}, {targetBone {boneName=upperarm_l}}]
    constraintWeights [1,1,1]
   }
   n116 = "Constraint - dyn_l_shoulder_leaf_011_master (PPP)"!(ComponentPose=@n115, Alpha=1, ConstraintWeights_0=0.8122, ConstraintWeights_1:float=0.1757, ConstraintWeights_2:float=0.0121) -> (Pose) {
    alphaBoolBlend {blendOption=Linear}
    boneToModify {boneName=dyn_l_shoulder_leaf_011_master}
    constraintSetup [{targetBone {boneName=upperarm_twist_01_l}}, {targetBone {boneName=clavicle_l}}, {targetBone {boneName=upperarm_l}}]
    constraintWeights [1,1,1]
   }
   n117 = "Constraint - dyn_l_shoulder_leaf_012_master (PPP)"!(ComponentPose=@n116, Alpha=1, ConstraintWeights_0=0.8472, ConstraintWeights_1:float=0.1021, ConstraintWeights_2:float=0.0508) -> (Pose) {
    alphaBoolBlend {blendOption=Linear}
    boneToModify {boneName=dyn_l_shoulder_leaf_012_master}
    constraintSetup [{targetBone {boneName=upperarm_twist_01_l}}, {targetBone {boneName=clavicle_l}}, {targetBone {boneName=upperarm_l}}]
    constraintWeights [1,1,1]
   }
   n118 = "Constraint - dyn_l_shoulder_leaf_013_master (PP)"!(ComponentPose=@n117, Alpha=1, ConstraintWeights_0=0.6601, ConstraintWeights_1:float=0.3399) -> (Pose) {
    alphaBoolBlend {blendOption=Linear}
    boneToModify {boneName=dyn_l_shoulder_leaf_013_master}
    constraintSetup [{targetBone {boneName=upperarm_twist_01_l}}, {targetBone {boneName=clavicle_l}}]
    constraintWeights [1,1]
   }
   n119 = "Constraint - dyn_l_shoulder_leaf_014_master (PPPP)"!(
    ComponentPose=@n118
    Alpha=1
    ConstraintWeights_0=0.0132
    ConstraintWeights_1:float=0.8195
    ConstraintWeights_2:float=0.1314
    ConstraintWeights_3:float=0.036
   ) -> (Pose) {
    alphaBoolBlend {blendOption=Linear}
    boneToModify {boneName=dyn_l_shoulder_leaf_014_master}
    constraintSetup [
     {targetBone {boneName=upperarm_twist_02_l}}
     {targetBone {boneName=upperarm_twist_01_l}}
     {targetBone {boneName=clavicle_l}}
     {targetBone {boneName=upperarm_l}}
    ]
    constraintWeights [1,1,1,1]
   }
   n120 = "Constraint - dyn_l_arm_flower_02_master (PPPPP)"!(
    ComponentPose=@n119
    Alpha=1
    ConstraintWeights_0=0.2583
    ConstraintWeights_1:float=0.1348
    ConstraintWeights_2:float=0.0119
    ConstraintWeights_3:float=0.595
    ConstraintWeights_4:float=0
   ) -> (Pose) {
    alphaBoolBlend {blendOption=Linear}
    boneToModify {boneName=dyn_l_arm_flower_02_master}
    constraintSetup [
     {targetBone {boneName=lowerarm_l}}
     {targetBone {boneName=lowerarm_twist_01_l}}
     {targetBone {boneName=upperarm_l}}
     {targetBone {boneName=lowerarm_twist_02_l}}
     {targetBone {boneName=upperarm_twist_02_l}}
    ]
    constraintWeights [1,1,1,1,1]
   }
   n121 = "Constraint - dyn_l_arm_flower_03_master (PPPP)"!(
    ComponentPose=@n120
    Alpha=1
    ConstraintWeights_0=0.1821
    ConstraintWeights_1:float=0.2261
    ConstraintWeights_2:float=0.0013
    ConstraintWeights_3:float=0.5906
   ) -> (Pose) {
    alphaBoolBlend {blendOption=Linear}
    boneToModify {boneName=dyn_l_arm_flower_03_master}
    constraintSetup [
     {targetBone {boneName=lowerarm_l}}
     {targetBone {boneName=lowerarm_twist_01_l}}
     {targetBone {boneName=upperarm_l}}
     {targetBone {boneName=lowerarm_twist_02_l}}
    ]
    constraintWeights [1,1,1,1]
   }
   n122 = "Constraint - dyn_l_shoulder_flower_05_master (PPPPP)"!(
    ComponentPose=@n121
    Alpha=1
    ConstraintWeights_0=0.0439
    ConstraintWeights_1:float=0
    ConstraintWeights_2:float=0.2373
    ConstraintWeights_3:float=0.7049
    ConstraintWeights_4:float=0.0139
   ) -> (Pose) {
    alphaBoolBlend {blendOption=Linear}
    boneToModify {boneName=dyn_l_shoulder_flower_05_master}
    constraintSetup [
     {targetBone {boneName=spine_03}}
     {targetBone {boneName=thigh_twist_01_l}}
     {targetBone {boneName=spine_05}}
     {targetBone {boneName=clavicle_l}}
     {targetBone {boneName=upperarm_twist_01_l}}
    ]
    constraintWeights [1,1,1,1,1]
   }
   n123 = "Constraint - dyn_l_back_shoulder_fern_master (PPPPPP)"!(
    ComponentPose=@n122
    Alpha=1
    ConstraintWeights_0=0.1443
    ConstraintWeights_1:float=0.7728
    ConstraintWeights_2:float=0
    ConstraintWeights_3:float=0.0152
    ConstraintWeights_4:float=0.0535
    ConstraintWeights_5:float=0.0141
   ) -> (Pose) {
    alphaBoolBlend {blendOption=Linear}
    boneToModify {boneName=dyn_l_back_shoulder_fern_master}
    constraintSetup [
     {targetBone {boneName=spine_03}}
     {targetBone {boneName=clavicle_l}}
     {targetBone {boneName=thigh_twist_01_l}}
     {targetBone {boneName=spine_04}}
     {targetBone {boneName=spine_05}}
     {targetBone {boneName=upperarm_twist_01_l}}
    ]
    constraintWeights [1,1,1,1,1,1]
   }
   n124 = "Constraint - dyn_back_vine_l_master (PPPPP)"!(
    ComponentPose=@n123
    Alpha=1
    ConstraintWeights_0=0.2651
    ConstraintWeights_1:float=0.074
    ConstraintWeights_2:float=0.0014
    ConstraintWeights_3:float=0.6591
    ConstraintWeights_4:float=0.0004
   ) -> (Pose) {
    alphaBoolBlend {blendOption=Linear}
    boneToModify {boneName=dyn_back_vine_l_master}
    constraintSetup [
     {targetBone {boneName=spine_03}}
     {targetBone {boneName=spine_04}}
     {targetBone {boneName=spine_05}}
     {targetBone {boneName=clavicle_l}}
     {targetBone {boneName=upperarm_twist_01_l}}
    ]
    constraintWeights [1,1,1,1,1]
   }
   n125 = "Constraint - dyn_l_leg_leaf_01_master (PPP)"!(ComponentPose=@n124, Alpha=1, ConstraintWeights_0=0.0006, ConstraintWeights_1:float=0.2415, ConstraintWeights_2:float=0.7578) -> (Pose) {
    alphaBoolBlend {blendOption=Linear}
    boneToModify {boneName=dyn_l_leg_leaf_01_master}
    constraintSetup [{targetBone {boneName=calf_l}}, {targetBone {boneName=thigh_twist_01_l}}, {targetBone {boneName=thigh_l}}]
    constraintWeights [1,1,1]
   }
   n126 = "Constraint - dyn_l_leg_leaf_02_master (PPPPP)"!(
    ComponentPose=@n125
    Alpha=1
    ConstraintWeights_0=0.0007
    ConstraintWeights_1:float=0.0068
    ConstraintWeights_2:float=0.7913
    ConstraintWeights_3:float=0.1425
    ConstraintWeights_4:float=0.0587
   ) -> (Pose) {
    alphaBoolBlend {blendOption=Linear}
    boneToModify {boneName=dyn_l_leg_leaf_02_master}
    constraintSetup [
     {targetBone {boneName=calf_l}}
     {targetBone {boneName=thigh_twist_01_l}}
     {targetBone {boneName=dyn_l_leg_leaf_02}}
     {targetBone {boneName=dyn_l_leg_vine_wobble_01}}
     {targetBone {boneName=thigh_l}}
    ]
    constraintWeights [1,1,1,1,1]
   }
   n127 = "Constraint - dyn_l_leg_leaf_03_master (PPPPP)"!(
    ComponentPose=@n126
    Alpha=1
    ConstraintWeights_0=0.0788
    ConstraintWeights_1:float=0.0003
    ConstraintWeights_2:float=0.6206
    ConstraintWeights_3:float=0.1427
    ConstraintWeights_4:float=0.1576
   ) -> (Pose) {
    alphaBoolBlend {blendOption=Linear}
    boneToModify {boneName=dyn_l_leg_leaf_03_master}
    constraintSetup [
     {targetBone {boneName=calf_l}}
     {targetBone {boneName=thigh_twist_01_l}}
     {targetBone {boneName=dyn_l_leg_leaf_03}}
     {targetBone {boneName=thigh_l}}
     {targetBone {boneName=dyn_l_leg_vine_wobble_01}}
    ]
    constraintWeights [1,1,1,1,1]
   }
   n128 = "Constraint - dyn_l_leg_leaf_04_master (PPP)"!(ComponentPose=@n127, Alpha=1, ConstraintWeights_0=0.8162, ConstraintWeights_1:float=0.0011, ConstraintWeights_2:float=0.1827) -> (Pose) {
    alphaBoolBlend {blendOption=Linear}
    boneToModify {boneName=dyn_l_leg_leaf_04_master}
    constraintSetup [{targetBone {boneName=calf_l}}, {targetBone {boneName=calf_twist_02_l}}, {targetBone {boneName=thigh_l}}]
    constraintWeights [1,1,1]
   }
   n129 = "Constraint - dyn_l_shoulder_leaf_015_master (PP)"!(ComponentPose=@n128, Alpha=1, ConstraintWeights_0=0.5819, ConstraintWeights_1:float=0.4181) -> (Pose) {
    alphaBoolBlend {blendOption=Linear}
    boneToModify {boneName=dyn_l_shoulder_leaf_015_master}
    constraintSetup [{targetBone {boneName=upperarm_twist_01_l}}, {targetBone {boneName=clavicle_l}}]
    constraintWeights [1,1]
   }
   n130 = "Constraint - dyn_r_shoulder_leaf_19_master (PPP)"!(ComponentPose=@n129, Alpha=1, ConstraintWeights_0=0.5198, ConstraintWeights_1:float=0.4802, ConstraintWeights_2:float=0) -> (Pose) {
    alphaBoolBlend {blendOption=Linear}
    boneToModify {boneName=dyn_r_shoulder_leaf_19_master}
    constraintSetup [{targetBone {boneName=upperarm_twist_01_r}}, {targetBone {boneName=clavicle_r}}, {targetBone {boneName=spine_03}}]
    constraintWeights [1,1,1]
   }
   n131 = "Constraint - dyn_r_bandana_master (PPP)"!(ComponentPose=@n130, Alpha=1, ConstraintWeights_0=0.1581, ConstraintWeights_1:float=0.5835, ConstraintWeights_2:float=0.2584) -> (Pose) {
    alphaBoolBlend {blendOption=Linear}
    boneToModify {boneName=dyn_r_bandana_master}
    constraintSetup [{targetBone {boneName=upperarm_twist_01_r}}, {targetBone {boneName=upperarm_r}}, {targetBone {boneName=upperarm_twist_02_r}}]
    constraintWeights [1,1,1]
   }
   n132 = Get crouching() -> (crouching:double)
   n133 = "Constraint - dyn_l_leg_leaf_04 (TT)"!(ComponentPose=@n131, Alpha=@n132, ConstraintWeights_0=1, ConstraintWeights_1:float=0) -> (Pose) {
    alphaBoolBlend {blendOption=Linear}
    boneToModify {boneName=dyn_l_leg_leaf_04}
    constraintSetup [{transformType=Translation, targetBone {boneName=calf_l}}, {transformType=Translation, targetBone {boneName=thigh_l}}]
    constraintWeights [1,1]
   }
   n134 = "Constraint - dyn_l_leg_leaf_04_b (TT)"!(ComponentPose=@n133, Alpha=@n132, ConstraintWeights_0=1, ConstraintWeights_1:float=0) -> (Pose) {
    alphaBoolBlend {blendOption=Linear}
    boneToModify {boneName=dyn_l_leg_leaf_04_b}
    constraintSetup [{transformType=Translation, targetBone {boneName=calf_l}}, {transformType=Translation, targetBone {boneName=thigh_l}}]
    constraintWeights [1,1]
   }
   n135 = "Transform (Modify) Bone - Bone: dyn_l_leg_leaf_04_master"(ComponentPose=@n134, Alpha=@n132, Translation="0,0,0", Rotation="60,0,0", Scale="1,1,1") {
    rotationMode=BMM_Additive
    scaleMode=BMM_Additive
    translationMode=BMM_Additive
    alphaBoolBlend {blendOption=Linear}
    boneToModify {boneName=dyn_l_leg_leaf_04_master}
    rotation {pitch=60}
   }
   n136 = Get backbling() -> (backbling:double)
   n137 = "Constraint - dyn_back_vine_l_01 (P)"(ComponentPose=@n135, Alpha=@n136, ConstraintWeights_0=1) {
    alphaBoolBlend {blendOption=Linear}
    boneToModify {boneName=dyn_back_vine_l_01}
    constraintSetup [{targetBone {boneName=spine_05}}]
    constraintWeights [1]
   }
   n138 = "Constraint - dyn_back_vine_l_02 (P)"(ComponentPose=@n137, Alpha=@n136, ConstraintWeights_0=1) {
    alphaBoolBlend {blendOption=Linear}
    boneToModify {boneName=dyn_back_vine_l_02}
    constraintSetup [{targetBone {boneName=spine_05}}]
    constraintWeights [1]
   }
   n139 = "Constraint - dyn_back_vine_l_03 (P)"(ComponentPose=@n138, Alpha=@n136, ConstraintWeights_0=1) {
    alphaBoolBlend {blendOption=Linear}
    boneToModify {boneName=dyn_back_vine_l_03}
    constraintSetup [{targetBone {boneName=spine_04}}]
    constraintWeights [1]
   }
   n140 = "Transform (Modify) Bone - Bone: dyn_l_leg_leaf_04_master"(ComponentPose=@n139, Alpha=Get leaf_shrink() -> (leaf_shrink:double), Translation="0,0,0", Rotation="0,0,0", Scale="0.5,0.5,0.5") {
    rotationMode=BMM_Additive
    scaleMode=BMM_Additive
    translationMode=BMM_Additive
    alphaBoolBlend {blendOption=Linear}
    boneToModify {boneName=dyn_l_leg_leaf_04_master}
    scale=<0.5 0.5 0.5>
   }
   Output Pose(Result=Component To Local(ComponentPose=@n140)) {layerGroup=None}
  ]}
 EventGraph {nodes [
   Event BlueprintUpdateAnimation()
   > n1 = Sequence()
   > Set crouching(crouching:double=SelectFloat(A=1, B=0, bPickA=Get bIsCrouching())) -> (Output_Get:double)
   > Set ADS_Aim(ADS_Aim:double=SelectFloat(A=0.5, B=1, bPickA=Get bIsTargeting())) -> (Output_Get:double)
   > n2 = ComplexSimSpaceFunction DEPRECATED!(
    execute
    CharacterPartAnimInstance:Custom Character Part Anim Instance Object Reference=Self-Reference()
    InterpSpeedMax:double=9
    PawnSpeedMax:double=20
    RInterpErrorTolerance:double=1
    DeltaTime:double=0
    DebugDraw:bool=false
    sim_space_rotation_in:Rotator=Get sim_space_rotation()
    inComponentSpaceIn:bool=Get InComponentSpace() -> (InComponentSpace:bool)
    inComponentSpaceLastFrameIn:bool=Get InComponentSpaceLastFrame() -> (InComponentSpaceLastFrame:bool)
    lockedOnSimSpaceIn:bool=Get LockedOnSimSpace() -> (LockedOnSimSpace:bool)
    InterpSpeedIn:double=Get InterpSpeed() -> (InterpSpeed:double)
   ) -> (
    then
    sim_space_rotation:Rotator
    inComponentSpaceOut:bool
    inComponentSpaceLastFrameOut:bool
    lockedOnSimSpaceOut:bool
    InterpSpeedOut:double
   )
   > Set sim_space_rotation(sim_space_rotation:Rotator=@n2) -> (Output_Get:Rotator)
   > Set InComponentSpace(InComponentSpace:bool=@n2.inComponentSpaceOut) -> (Output_Get:bool)
   > Set InComponentSpaceLastFrame(InComponentSpaceLastFrame:bool=@n2.inComponentSpaceLastFrameOut) -> (Output_Get:bool)
   > Set LockedOnSimSpace(LockedOnSimSpace:bool=@n2.lockedOnSimSpaceOut) -> (Output_Get:bool)
   > Set InterpSpeed(InterpSpeed:double=@n2.InterpSpeedOut) -> (Output_Get:double)
   n3 = Branch(execute=@n1.then_1, Condition=Get bIsCrouching())
   n4 = BreakTransform(InTransform=GetDeltaTransformFromRefPose(self=GetOwnerSkeletalMeshComponent(), BoneName=calf_l, BaseName=thigh_l))
   > Set leaf_shrink(leaf_shrink:double=MapRangeClamped(Value=BreakRotator(InRot=@n4.Rotation).Yaw, InRangeA=75, InRangeB=105)) -> (Output_Get:double)
   Set leaf_shrink(execute=@n3.else, leaf_shrink=0)
   n5 = Event OnEndPlayingEmote() {is=title K2Node_Event}
   n6 = Delay(execute=[@n5.then, @n7.then], Duration=0.1)
   ResetDynamics(execute=[@n6.then, @n9.then], InTeleportType=TeleportPhysics)
   > ResetDynamics(InTeleportType=ResetPhysics)
   n7 = Event OnEndSkydiving()
   Event OnCharacterCustomizationLoaded()
   n8 = CheckForBackbling()
   > Set backbling(backbling:double=SelectFloat(A=0.5, B=0, bPickA=OR Boolean(A=@n8, B=@n8.bIsBackblingLong))) -> (Output_Get:double)
   > ResetDynamics(InTeleportType=TeleportPhysics)
   > ResetDynamics(InTeleportType=ResetPhysics)
   Event BlueprintInitializeAnimation()
   > n9 = Delay(Duration=0.15)
  ]}
]
variables [
 ADS_Aim {type=number}
 backbling {type=number}
 crouching {type=number}
 InComponentSpace {type=boolean}
 InComponentSpaceLastFrame {type=boolean}
 InterpSpeed {type=number}
 leaf_shrink {type=number}
 LockedOnSimSpace {type=boolean}
 sim_space_rotation {type=Rotator}
]
