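# Physics asset for the Moss Knight player body cosmetic that gives its three flowers and seven leaves secondary motion around the head.
# A head capsule anchors small master capsules, one per flower or leaf, through constraints with free swing and twist.
# Each master carries a box or capsule body with narrow, individually tuned angular limits.
# Leaves 01, 03 and 05 chain a second segment for extra bend.
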PhysicsAsset /BRCosmetics/Characters/Player/Male/Medium/Bodies/M_MED_Moss_Knight/Meshes/Parts/M_MED_Moss_Knight_Physics
bodies {
 dyn_flower_01 [{shapeType=Box, extentX=0.6861, extentY=2.9789, extentZ=2.9789, center=<0.4353 -0.0001 0>}]
 dyn_flower_01_master [{shapeType=Capsule, length=1, radius=1}]
 dyn_flower_02 [{shapeType=Box, extentX=0.6861, extentY=2.9789, extentZ=2.9789, center=<0.4353 -0.0001 0>}]
 dyn_flower_02_master [{shapeType=Capsule, length=1, radius=1}]
 dyn_flower_03 [{shapeType=Box, extentX=0.6861, extentY=2.9789, extentZ=2.9789, center=<0.4353 -0.0001 0>}]
 dyn_flower_03_master [{shapeType=Capsule, length=1, radius=1}]
 dyn_head [{shapeType=Capsule, length=16.4455, radius=9.6619, center=<-1.3124 2.0751 0>, rotation=<rot 90 0 0>}]
 dyn_leaf_01 [{shapeType=Box, extentX=3.8855, extentY=2, extentZ=0.8711, center=<2.1265 0.0002 0>}]
 dyn_leaf_01_b [{shapeType=Box, extentX=3.8855, extentY=2, extentZ=0.8711, center=<2.1265 0.0002 0>}]
 dyn_leaf_01_master [{shapeType=Capsule, length=1, radius=1}]
 dyn_leaf_02 [{shapeType=Capsule, length=1.639, radius=0.575, center=<1.4475 0.0077 0.0419>, rotation=<rot 0 -89.6944 -86.7412>}]
 dyn_leaf_02_master [{shapeType=Capsule, length=1, radius=1}]
 dyn_leaf_03 [{shapeType=Box, extentX=3.8855, extentY=2, extentZ=0.8711, center=<2.1265 0.0002 0>}]
 dyn_leaf_03_b [{shapeType=Capsule, length=1.596, radius=1.2051, center=<1.5117 -0.049 -0.1208>, rotation=<rot 0 -90.4486 -85.7111>}]
 dyn_leaf_03_master [{shapeType=Capsule, length=1, radius=1}]
 dyn_leaf_04 [{shapeType=Box, extentX=3.8855, extentY=2, extentZ=0.8711, center=<2.1265 0.0002 0>}]
 dyn_leaf_04_master [{shapeType=Capsule, length=1, radius=1}]
 dyn_leaf_05 [{shapeType=Box, extentX=3.8855, extentY=2, extentZ=0.8711, center=<2.1265 0.0002 0>}]
 dyn_leaf_05_b [{shapeType=Box, extentX=3.8855, extentY=2, extentZ=0.8711, center=<2.1265 0.0002 0>}]
 dyn_leaf_05_master [{shapeType=Capsule, length=1, radius=1}]
 dyn_leaf_06 [{shapeType=Capsule, length=1.6808, radius=0.5456, center=<1.4059 -0.1836 0.0133>, rotation=<rot 0 -97.4943 -86.7474>}]
 dyn_leaf_06_master [{shapeType=Capsule, length=1, radius=1}]
 dyn_leaf_07 [{shapeType=Box, extentX=3.8855, extentY=2, extentZ=0.8711, center=<2.1265 0.0002 0>}]
 dyn_leaf_07_master [{shapeType=Capsule, length=1, radius=1}]
}
constraints [
 {
  bone1Name=dyn_flower_01_master
  bone2Name=dyn_head
  swing1LimitDegrees=45
  swing1Motion=Free
  swing2LimitDegrees=45
  swing2Motion=Free
  twistLimitDegrees=45
  twistMotion=Free
 }
 {bone1Name=dyn_flower_01, bone2Name=dyn_flower_01_master, swing1LimitDegrees=15, swing2LimitDegrees=15, twistLimitDegrees=15}
 {
  bone1Name=dyn_leaf_01_master
  bone2Name=dyn_head
  swing1LimitDegrees=45
  swing1Motion=Free
  swing2LimitDegrees=45
  swing2Motion=Free
  twistLimitDegrees=45
  twistMotion=Free
 }
 {bone1Name=dyn_leaf_01, bone2Name=dyn_leaf_01_master, swing1LimitDegrees=25, swing2LimitDegrees=25, twistLimitDegrees=5}
 {bone1Name=dyn_leaf_01_b, bone2Name=dyn_leaf_01, swing1LimitDegrees=25, swing2LimitDegrees=40, twistLimitDegrees=5}
 {
  bone1Name=dyn_leaf_02_master
  bone2Name=dyn_head
  swing1LimitDegrees=45
  swing1Motion=Free
  swing2LimitDegrees=45
  swing2Motion=Free
  twistLimitDegrees=45
  twistMotion=Free
 }
 {bone1Name=dyn_leaf_02, bone2Name=dyn_leaf_02_master, swing1LimitDegrees=20, swing2LimitDegrees=8, twistLimitDegrees=5}
 {
  bone1Name=dyn_leaf_03_master
  bone2Name=dyn_head
  swing1LimitDegrees=45
  swing1Motion=Free
  swing2LimitDegrees=45
  swing2Motion=Free
  twistLimitDegrees=45
  twistMotion=Free
 }
 {bone1Name=dyn_leaf_03, bone2Name=dyn_leaf_03_master, swing1LimitDegrees=10, swing2LimitDegrees=10, twistLimitDegrees=5}
 {bone1Name=dyn_leaf_03_b, bone2Name=dyn_leaf_03, swing1LimitDegrees=15, swing2LimitDegrees=25, twistLimitDegrees=5}
 {
  bone1Name=dyn_leaf_05_master
  bone2Name=dyn_head
  swing1LimitDegrees=45
  swing1Motion=Free
  swing2LimitDegrees=45
  swing2Motion=Free
  twistLimitDegrees=45
  twistMotion=Free
 }
 {bone1Name=dyn_leaf_05, bone2Name=dyn_leaf_05_master, swing1LimitDegrees=25, swing2LimitDegrees=25, twistLimitDegrees=5}
 {bone1Name=dyn_leaf_05_b, bone2Name=dyn_leaf_05, swing1LimitDegrees=25, swing2LimitDegrees=40, twistLimitDegrees=5}
 {
  bone1Name=dyn_flower_02_master
  bone2Name=dyn_head
  swing1LimitDegrees=45
  swing1Motion=Free
  swing2LimitDegrees=45
  swing2Motion=Free
  twistLimitDegrees=45
  twistMotion=Free
 }
 {bone1Name=dyn_flower_02, bone2Name=dyn_flower_02_master, swing1LimitDegrees=15, swing2LimitDegrees=15, twistLimitDegrees=15}
 {
  bone1Name=dyn_flower_03_master
  bone2Name=dyn_head
  swing1LimitDegrees=45
  swing1Motion=Free
  swing2LimitDegrees=45
  swing2Motion=Free
  twistLimitDegrees=45
  twistMotion=Free
 }
 {bone1Name=dyn_flower_03, bone2Name=dyn_flower_03_master, swing1LimitDegrees=15, swing2LimitDegrees=15, twistLimitDegrees=15}
 {
  bone1Name=dyn_leaf_07_master
  bone2Name=dyn_head
  swing1LimitDegrees=45
  swing1Motion=Free
  swing2LimitDegrees=45
  swing2Motion=Free
  twistLimitDegrees=45
  twistMotion=Free
 }
 {bone1Name=dyn_leaf_07, bone2Name=dyn_leaf_07_master, swing1LimitDegrees=20, swing2LimitDegrees=20, twistLimitDegrees=5}
 {
  bone1Name=dyn_leaf_06_master
  bone2Name=dyn_head
  swing1LimitDegrees=45
  swing1Motion=Free
  swing2LimitDegrees=45
  swing2Motion=Free
  twistLimitDegrees=45
  twistMotion=Free
 }
 {bone1Name=dyn_leaf_06, bone2Name=dyn_leaf_06_master, swing1LimitDegrees=5, swing2LimitDegrees=5, twistLimitDegrees=5}
 {
  bone1Name=dyn_leaf_04_master
  bone2Name=dyn_head
  swing1LimitDegrees=45
  swing1Motion=Free
  swing2LimitDegrees=45
  swing2Motion=Free
  twistLimitDegrees=45
  twistMotion=Free
 }
 {bone1Name=dyn_leaf_04, bone2Name=dyn_leaf_04_master, swing1LimitDegrees=20, swing2LimitDegrees=5, twistLimitDegrees=5}
]
